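# Flipbook anime face material for the PacificSweater large male character body, built on the 2023 anime character shader functions.
# Flipbook face color and a custom eye-animation function are composited into lit and shaded colors before the shader's post stage.
# Animation-driven eye indices equal to 3 switch masks for blink recoloring and eye shadows.
# It also carries a vertex-color-masked "Ozone Head" position offset and temporal responsiveness signals with an emissive debug view.
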Material /BRCosmetics/Characters/Player/Male/Large/Bodies/M_LRG_PacificSweater/Materials/M_FlipbookFace_PacificSweater
graph {
 expressions {
  MaterialExpressionAdd_0 {
   class=MaterialExpressionAdd
   inputs [
    A {from=MaterialExpressionGetMaterialAttributes_1, output=World Position Offset}
    B {from=MaterialExpressionGetMaterialAttributes_0, output=World Position Offset}
   ]
  }
  MaterialExpressionAdd_1 {
   class=MaterialExpressionAdd
   inputs [
    A {from=MaterialExpressionStaticSwitch_2}
    B {from=MaterialExpressionAppendVector_5}
   ]
  }
  MaterialExpressionAdd_2 {
   class=MaterialExpressionAdd
   inputs [
    A {from=MaterialExpressionAppendVector_0}
    B {from=MaterialExpressionStaticSwitchParameter_4}
   ]
  }
  MaterialExpressionAdd_3 {
   class=MaterialExpressionAdd
   inputs [
    A {from=MaterialExpressionConvert_0}
    B {from=MaterialExpressionGetMaterialAttributes_1, output=Emissive Color}
   ]
  }
  MaterialExpressionAppendVector_0 {
   class=MaterialExpressionAppendVector
   inputs [
    A {from=MaterialExpressionScalarParameter_6, output=""}
    B {from=MaterialExpressionScalarParameter_7, output=""}
   ]
  }
  MaterialExpressionAppendVector_13 {
   class=MaterialExpressionAppendVector
   inputs [
    A {from=MaterialExpressionScalarParameter_10, output=""}
    B {from=MaterialExpressionScalarParameter_15, output=""}
   ]
  }
  MaterialExpressionAppendVector_5 {
   class=MaterialExpressionAppendVector
   inputs [
    A {from=MaterialExpressionScalarParameter_11, output=""}
    B {from=MaterialExpressionScalarParameter_12, output=""}
   ]
  }
  MaterialExpressionComponentMask_0 {class=MaterialExpressionComponentMask, inputs [None {from=MaterialExpressionTextureSampleParameter2D_3, output=R}], properties {r=true}}
  MaterialExpressionComponentMask_3 {class=MaterialExpressionComponentMask, inputs [None {from=MaterialExpressionTextureSampleParameter2D_3, output=G}], properties {g=true}}
  MaterialExpressionConstant4Vector_0 {class=MaterialExpressionConstant4Vector, inputs [], properties {constant=<rgba 0.9301 0.6445 0.3864 1>}}
  MaterialExpressionConstant4Vector_1 {class=MaterialExpressionConstant4Vector, inputs [], properties {constant=<rgba 1 0.7157 0.4621 1>}}
  MaterialExpressionConstant_0 {class=MaterialExpressionConstant, properties {r=1}}
  MaterialExpressionConstant_1 {class=MaterialExpressionConstant}
  MaterialExpressionConstant_10 {class=MaterialExpressionConstant}
  MaterialExpressionConstant_11 {class=MaterialExpressionConstant, properties {r=1}}
  MaterialExpressionConstant_12 {class=MaterialExpressionConstant}
  MaterialExpressionConstant_14 {class=MaterialExpressionConstant}
  MaterialExpressionConstant_28 {class=MaterialExpressionConstant}
  MaterialExpressionConstant_29 {class=MaterialExpressionConstant, properties {r=1}}
  MaterialExpressionConstant_30 {class=MaterialExpressionConstant}
  MaterialExpressionConstant_5 {class=MaterialExpressionConstant}
  MaterialExpressionConstant_6 {class=MaterialExpressionConstant, properties {r=1}}
  MaterialExpressionConstant_7 {class=MaterialExpressionConstant}
  MaterialExpressionConstant_8 {class=MaterialExpressionConstant, properties {r=1}}
  MaterialExpressionConstant_9 {class=MaterialExpressionConstant, properties {r=1}}
  MaterialExpressionConvert_0 {
   class=MaterialExpressionConvert
   inputs [
    None {from=MaterialExpressionScalarParameter_2}
    None {from=""}
    None {from=MaterialExpressionScalarParameter_13}
   ]
   properties {
    nodeName=Make Vector3
    convertInputs [
     {
      type=Scalar
      expressionInput {
       expression=&self:MaterialExpressionScalarParameter_2
       inputName=None
       mask=0
       maskA=0
       maskB=0
       maskG=0
       maskR=0
       outputIndex=0
      }
     }
     {type=Scalar, expressionInput {expression=null, inputName=None, mask=0, maskA=0, maskB=0, maskG=0, maskR=0, outputIndex=0}}
     {
      type=Scalar
      expressionInput {
       expression=&self:MaterialExpressionScalarParameter_13
       inputName=None
       mask=0
       maskA=0
       maskB=0
       maskG=0
       maskR=0
       outputIndex=0
      }
     }
    ]
    convertOutputs [{type=Vector3}]
   }
  }
  MaterialExpressionConvert_1 {
   class=MaterialExpressionConvert
   inputs [None {from=MaterialExpressionNamedRerouteUsage_9}]
   properties {
    nodeName=Break Float3
    convertInputs [{
      type=Vector3
      expressionInput {
       expression=&self:MaterialExpressionNamedRerouteUsage_9
       inputName=None
       mask=0
       maskA=0
       maskB=0
       maskG=0
       maskR=0
       outputIndex=0
      }
     }]
    convertOutputs [{type=Scalar}, {type=Scalar}, {type=Scalar}]
   }
  }
  MaterialExpressionGetMaterialAttributes_0 {
   class=MaterialExpressionGetMaterialAttributes
   inputs [None {from=MaterialExpressionNamedRerouteUsage_23, output=""}]
   properties {attributeGetTypes [#1]}
  }
  MaterialExpressionGetMaterialAttributes_1 {
   class=MaterialExpressionGetMaterialAttributes
   inputs [None {from=MaterialExpressionMaterialFunctionCall_14, output=Result}]
   properties {attributeGetTypes [#1, #2]}
  }
  MaterialExpressionIf_0 {
   class=MaterialExpressionIf
   inputs [
    A {from=MaterialExpressionVertexColor_0, output=R}
    B {from="", output=""}
    "A > B" {from=MaterialExpressionConstant_0, output=""}
    "A == B" {from=MaterialExpressionConstant_0, output=""}
    "A < B" {from=MaterialExpressionConstant_1, output=""}
   ]
   properties {constB=0.01}
  }
  MaterialExpressionIf_1 {
   class=MaterialExpressionIf
   inputs [
    A {from=MaterialExpressionScalarParameter_18, output=""}
    B {from="", output=""}
    "A > B" {from=MaterialExpressionNamedRerouteUsage_34, output=""}
    "A == B" {from=MaterialExpressionNamedRerouteUsage_36, output=""}
    "A < B" {from=MaterialExpressionNamedRerouteUsage_34, output=""}
   ]
   properties {constB=3}
  }
  MaterialExpressionIf_10 {
   class=MaterialExpressionIf
   inputs [
    A {from=MaterialExpressionScalarParameter_29, output=""}
    B {from="", output=""}
    "A > B" {from=MaterialExpressionConstant_12, output=""}
    "A == B" {from=MaterialExpressionNamedRerouteUsage_56, output=""}
    "A < B" {from=MaterialExpressionConstant_12, output=""}
   ]
   properties {constB=3}
  }
  MaterialExpressionIf_12 {
   class=MaterialExpressionIf
   inputs [
    A {from=MaterialExpressionScalarParameter_30, output=""}
    B {from=MaterialExpressionScalarParameter_32, output=""}
    "A > B" {from=MaterialExpressionConstant_28, output=""}
    "A == B" {from=MaterialExpressionScalarParameter_30, output=""}
    "A < B" {from=MaterialExpressionConstant_28, output=""}
   ]
   properties {}
  }
  MaterialExpressionIf_13 {
   class=MaterialExpressionIf
   inputs [
    A {from=MaterialExpressionIf_12, output=""}
    B {from="", output=""}
    "A > B" {from=MaterialExpressionConstant_30, output=""}
    "A == B" {from=MaterialExpressionConstant_29, output=""}
    "A < B" {from=MaterialExpressionConstant_30, output=""}
   ]
   properties {constB=3}
  }
  MaterialExpressionIf_2 {
   class=MaterialExpressionIf
   inputs [
    A {from=MaterialExpressionScalarParameter_19, output=""}
    B {from="", output=""}
    "A > B" {from=MaterialExpressionNamedRerouteUsage_38, output=""}
    "A == B" {from=MaterialExpressionNamedRerouteUsage_39, output=""}
    "A < B" {from=MaterialExpressionNamedRerouteUsage_38, output=""}
   ]
   properties {constB=3}
  }
  MaterialExpressionIf_3 {
   class=MaterialExpressionIf
   inputs [
    A {from=MaterialExpressionScalarParameter_20, output=""}
    B {from="", output=""}
    "A > B" {from=MaterialExpressionNamedRerouteUsage_40, output=""}
    "A == B" {from=MaterialExpressionNamedRerouteUsage_41, output=""}
    "A < B" {from=MaterialExpressionNamedRerouteUsage_40, output=""}
   ]
   properties {constB=3}
  }
  MaterialExpressionIf_4 {
   class=MaterialExpressionIf
   inputs [
    A {from=MaterialExpressionScalarParameter_21, output=""}
    B {from="", output=""}
    "A > B" {from=MaterialExpressionNamedRerouteUsage_42, output=""}
    "A == B" {from=MaterialExpressionNamedRerouteUsage_43, output=""}
    "A < B" {from=MaterialExpressionNamedRerouteUsage_42, output=""}
   ]
   properties {constB=3}
  }
  MaterialExpressionIf_7 {
   class=MaterialExpressionIf
   inputs [
    A {from=MaterialExpressionScalarParameter_26, output=""}
    B {from="", output=""}
    "A > B" {from=MaterialExpressionNamedRerouteUsage_46, output=""}
    "A == B" {from=MaterialExpressionNamedRerouteUsage_47, output=""}
    "A < B" {from=MaterialExpressionNamedRerouteUsage_46, output=""}
   ]
   properties {constB=3}
  }
  MaterialExpressionIf_8 {
   class=MaterialExpressionIf
   inputs [
    A {from=MaterialExpressionScalarParameter_27, output=""}
    B {from="", output=""}
    "A > B" {from=MaterialExpressionNamedRerouteUsage_48, output=""}
    "A == B" {from=MaterialExpressionNamedRerouteUsage_49, output=""}
    "A < B" {from=MaterialExpressionNamedRerouteUsage_48, output=""}
   ]
   properties {constB=3}
  }
  MaterialExpressionIf_9 {
   class=MaterialExpressionIf
   inputs [
    A {from=MaterialExpressionScalarParameter_28, output=""}
    B {from="", output=""}
    "A > B" {from=MaterialExpressionConstant_14, output=""}
    "A == B" {from=MaterialExpressionNamedRerouteUsage_51, output=""}
    "A < B" {from=MaterialExpressionConstant_14, output=""}
   ]
   properties {constB=3}
  }
  MaterialExpressionLinearInterpolate_0 {
   class=MaterialExpressionLinearInterpolate
   inputs [
    A {from=MaterialExpressionNamedRerouteUsage_20}
    B {from=MaterialExpressionNamedRerouteUsage_26}
    Alpha {from=MaterialExpressionReroute_3}
   ]
  }
  MaterialExpressionLinearInterpolate_1 {
   class=MaterialExpressionLinearInterpolate
   inputs [
    A {from=MaterialExpressionNamedRerouteUsage_21}
    B {from=MaterialExpressionNamedRerouteUsage_25}
    Alpha {from=MaterialExpressionSaturate_0}
   ]
  }
  MaterialExpressionLinearInterpolate_10 {
   class=MaterialExpressionLinearInterpolate
   inputs [
    A {from=MaterialExpressionAdd_0}
    B {from=MaterialExpressionSubtract_4}
    Alpha {from=MaterialExpressionScalarParameter_22}
   ]
  }
  MaterialExpressionLinearInterpolate_11 {
   class=MaterialExpressionLinearInterpolate
   inputs [
    A {from=MaterialExpressionNamedRerouteUsage_16}
    B {from=MaterialExpressionMultiply_1}
    Alpha {from=MaterialExpressionNamedRerouteUsage_15}
   ]
  }
  MaterialExpressionLinearInterpolate_12 {
   class=MaterialExpressionLinearInterpolate
   inputs [
    A {from=MaterialExpressionNamedRerouteUsage_17}
    B {from=MaterialExpressionMultiply_2}
    Alpha {from=MaterialExpressionNamedRerouteUsage_27}
   ]
  }
  MaterialExpressionLinearInterpolate_13 {
   class=MaterialExpressionLinearInterpolate
   inputs [
    A {from=""}
    B {from=MaterialExpressionTextureSampleParameter2D_1, output=G}
    Alpha {from=MaterialExpressionScalarParameter_9}
   ]
   properties {constA=1}
  }
  MaterialExpressionLinearInterpolate_14 {
   class=MaterialExpressionLinearInterpolate
   inputs [
    A {from=MaterialExpressionGetMaterialAttributes_1, output=Emissive Color}
    B {from=MaterialExpressionAdd_3}
    Alpha {from=MaterialExpressionScalarParameter_16}
   ]
  }
  MaterialExpressionLinearInterpolate_15 {
   class=MaterialExpressionLinearInterpolate
   inputs [
    A {from=MaterialExpressionLinearInterpolate_8}
    B {from=MaterialExpressionVectorParameter_0, output=RGB}
    Alpha {from=MaterialExpressionMultiply_6}
   ]
  }
  MaterialExpressionLinearInterpolate_16 {
   class=MaterialExpressionLinearInterpolate
   inputs [
    A {from=MaterialExpressionLinearInterpolate_21}
    B {from=MaterialExpressionMultiply_13}
    Alpha {from=MaterialExpressionIf_1}
   ]
  }
  MaterialExpressionLinearInterpolate_17 {
   class=MaterialExpressionLinearInterpolate
   inputs [
    A {from=MaterialExpressionLinearInterpolate_9}
    B {from=MaterialExpressionMultiply_12}
    Alpha {from=MaterialExpressionMultiply_6}
   ]
  }
  MaterialExpressionLinearInterpolate_2 {
   class=MaterialExpressionLinearInterpolate
   inputs [
    A {from=MaterialExpressionNamedRerouteUsage_22}
    B {from=MaterialExpressionMaterialFunctionCall_1, output=FaceColor}
    Alpha {from=MaterialExpressionPower_0}
   ]
  }
  MaterialExpressionLinearInterpolate_21 {
   class=MaterialExpressionLinearInterpolate
   inputs [
    A {from=MaterialExpressionLinearInterpolate_6}
    B {from=MaterialExpressionMultiply_13}
    Alpha {from=MaterialExpressionIf_2}
   ]
  }
  MaterialExpressionLinearInterpolate_22 {
   class=MaterialExpressionLinearInterpolate
   inputs [
    A {from=MaterialExpressionLinearInterpolate_23}
    B {from=MaterialExpressionMultiply_11}
    Alpha {from=MaterialExpressionIf_3}
   ]
  }
  MaterialExpressionLinearInterpolate_23 {
   class=MaterialExpressionLinearInterpolate
   inputs [
    A {from=MaterialExpressionLinearInterpolate_7}
    B {from=MaterialExpressionMultiply_11}
    Alpha {from=MaterialExpressionIf_4}
   ]
  }
  MaterialExpressionLinearInterpolate_24 {
   class=MaterialExpressionLinearInterpolate
   inputs [
    A {from=MaterialExpressionLinearInterpolate_25}
    B {from=""}
    Alpha {from=MaterialExpressionIf_7}
   ]
  }
  MaterialExpressionLinearInterpolate_25 {
   class=MaterialExpressionLinearInterpolate
   inputs [
    A {from=MaterialExpressionNamedRerouteUsage_31}
    B {from=""}
    Alpha {from=MaterialExpressionIf_8}
   ]
  }
  MaterialExpressionLinearInterpolate_26 {
   class=MaterialExpressionLinearInterpolate
   inputs [
    A {from=MaterialExpressionLinearInterpolate_27}
    B {from=MaterialExpressionSubtract_1}
    Alpha {from=MaterialExpressionIf_9}
   ]
  }
  MaterialExpressionLinearInterpolate_27 {
   class=MaterialExpressionLinearInterpolate
   inputs [
    A {from=MaterialExpressionConstant_5}
    B {from=MaterialExpressionSubtract_0}
    Alpha {from=MaterialExpressionIf_10}
   ]
  }
  MaterialExpressionLinearInterpolate_28 {
   class=MaterialExpressionLinearInterpolate
   inputs [
    A {from=MaterialExpressionLinearInterpolate_15}
    B {from=MaterialExpressionVectorParameter_2, output=RGB}
    Alpha {from=MaterialExpressionNamedRerouteUsage_32}
   ]
   properties {constB=0.05}
  }
  MaterialExpressionLinearInterpolate_29 {
   class=MaterialExpressionLinearInterpolate
   inputs [
    A {from=MaterialExpressionLinearInterpolate_17}
    B {from=MaterialExpressionVectorParameter_2, output=RGB}
    Alpha {from=MaterialExpressionNamedRerouteUsage_33}
   ]
   properties {constB=0.05}
  }
  MaterialExpressionLinearInterpolate_3 {
   class=MaterialExpressionLinearInterpolate
   inputs [
    A {from=""}
    B {from=MaterialExpressionLinearInterpolate_5}
    Alpha {from=MaterialExpressionMax_3}
   ]
   properties {constA=0.4}
  }
  MaterialExpressionLinearInterpolate_30 {
   class=MaterialExpressionLinearInterpolate
   inputs [
    A {from=MaterialExpressionLinearInterpolate_26}
    B {from=""}
    Alpha {from=MaterialExpressionIf_13}
   ]
  }
  MaterialExpressionLinearInterpolate_4 {
   class=MaterialExpressionLinearInterpolate
   inputs [
    A {from=""}
    B {from=MaterialExpressionScalarParameter_1}
    Alpha {from=MaterialExpressionTextureSampleParameter2D_1, output=R}
   ]
   properties {constA=1}
  }
  MaterialExpressionLinearInterpolate_5 {
   class=MaterialExpressionLinearInterpolate
   inputs [
    A {from=""}
    B {from=""}
    Alpha {from=MaterialExpressionMax_1}
   ]
   properties {constA=0.4}
  }
  MaterialExpressionLinearInterpolate_6 {
   class=MaterialExpressionLinearInterpolate
   inputs [
    A {from=MaterialExpressionLinearInterpolate_0}
    B {from=""}
    Alpha {from=MaterialExpressionNamedRerouteUsage_8}
   ]
  }
  MaterialExpressionLinearInterpolate_7 {
   class=MaterialExpressionLinearInterpolate
   inputs [
    A {from=MaterialExpressionLinearInterpolate_1}
    B {from=""}
    Alpha {from=MaterialExpressionNamedRerouteUsage_18}
   ]
  }
  MaterialExpressionLinearInterpolate_8 {
   class=MaterialExpressionLinearInterpolate
   inputs [
    A {from=MaterialExpressionMaterialFunctionCall_2, output=Lit}
    B {from=""}
    Alpha {from=MaterialExpressionNamedRerouteUsage_13}
   ]
   properties {constB=0.02}
  }
  MaterialExpressionLinearInterpolate_9 {
   class=MaterialExpressionLinearInterpolate
   inputs [
    A {from=MaterialExpressionMaterialFunctionCall_2, output=Shaded}
    B {from=""}
    Alpha {from=MaterialExpressionNamedRerouteUsage_14}
   ]
   properties {constB=0.02}
  }
  MaterialExpressionLocalPosition_0 {class=MaterialExpressionLocalPosition, inputs [], properties {localOrigin=InstancePreSkinning}}
  MaterialExpressionMaterialFunctionCall_0 {
   class=MaterialExpressionMaterialFunctionCall
   inputs [
    UVs {from=MaterialExpressionAdd_1}
    Texture Scale {from=MaterialExpressionNamedRerouteDeclaration_22}
   ]
   properties {materialFunction=&ScaleUVsByCenter}
  }
  MaterialExpressionMaterialFunctionCall_1 {
   class=MaterialExpressionMaterialFunctionCall
   inputs [
    Mouth UV {from=""}
    Eye UV {from=""}
    Brow UV {from=""}
    Material {from=""}
   ]
   properties {materialFunction=/Game/Global/Materials/Character_Parent_2023/FlipbookFace_Parent_2023/MaterialFunctions/MFI_FlipbookFace_Anime}
  }
  MaterialExpressionMaterialFunctionCall_14 {
   class=MaterialExpressionMaterialFunctionCall
   inputs [
    Shading_Gradient {from=MaterialExpressionMaterialFunctionCall_20, output=Shading_Gradient}
    Threshold_Gradient {from=MaterialExpressionMaterialFunctionCall_20, output=Shading_Gradient}
    Specular_Gradient {from=MaterialExpressionMaterialFunctionCall_20, output=Shading_Gradient}
    Sheen_Mask {from=MaterialExpressionMaterialFunctionCall_20, output=Shading_Gradient}
    Thickness {from=MaterialExpressionMaterialFunctionCall_20, output=Shading_Gradient}
    DFL_Mask {from=MaterialExpressionMaterialFunctionCall_20, output=Shading_Gradient}
    FX {from=MaterialExpressionMaterialFunctionCall_20, output=Shading_Gradient}
    Metallic_Gradient {from=MaterialExpressionMaterialFunctionCall_20, output=Shading_Gradient}
    RimLight_Gradient {from=MaterialExpressionMaterialFunctionCall_20, output=Shading_Gradient}
    NormalBuffer {from=MaterialExpressionMaterialFunctionCall_20, output=Shading_Gradient}
    -------- {from=""}
    CINE_LightSources {from=MaterialExpressionMaterialFunctionCall_20, output=Shading_Gradient}
    CINE_LightColor {from=MaterialExpressionMaterialFunctionCall_20, output=Shading_Gradient}
    --------- {from=""}
    Color_Lit {from=MaterialExpressionLinearInterpolate_11}
    Color_Shaded {from=MaterialExpressionReroute_0}
    SRM {from=MaterialExpressionMaterialFunctionCall_21, output=SRM}
    Emissive {from=MaterialExpressionReroute_2}
    DFL_Color {from=MaterialExpressionMaterialFunctionCall_21, output=SRM}
    Specular_Color {from=MaterialExpressionMaterialFunctionCall_21, output=SRM}
    RimLight_Color {from=MaterialExpressionMaterialFunctionCall_21, output=SRM}
   ]
   properties {materialFunction=/Game/Global/Materials/Character_Parent_2023/Anime_Parent_2023/MaterialFunctions/MFC_AnimeShader_2023_Post}
  }
  MaterialExpressionMaterialFunctionCall_2 {
   class=MaterialExpressionMaterialFunctionCall
   inputs [
    Color Lit {from=MaterialExpressionLinearInterpolate_16}
    Color Shaded {from=MaterialExpressionLinearInterpolate_22}
    Use UV2 {from=MaterialExpressionStaticBoolParameter_0}
    Mask {from=MaterialExpressionNamedRerouteUsage_19}
   ]
   properties {materialFunction=&/MF_AnimeEyeAnimation_PacificSweater}
  }
  MaterialExpressionMaterialFunctionCall_20 {
   class=MaterialExpressionMaterialFunctionCall
   inputs [
    CustomLightVector {from=""}
    OverrideNormal_WS {from=""}
    STT {from=MaterialExpressionMaterialFunctionCall_21, output=STT}
    Normal_Map {from=MaterialExpressionMaterialFunctionCall_21, output=STT}
    DFL {from=MaterialExpressionMaterialFunctionCall_21, output=STT}
    FX {from=MaterialExpressionMaterialFunctionCall_21, output=STT}
   ]
   properties {materialFunction=/Game/Global/Materials/Character_Parent_2023/Anime_Parent_2023/MaterialFunctions/MFC_AnimeShader_2023_Calculations}
  }
  MaterialExpressionMaterialFunctionCall_21 {
   class=MaterialExpressionMaterialFunctionCall
   inputs [CustomUV {from=""}]
   properties {materialFunction=/Game/Global/Materials/Character_Parent_2023/Anime_Parent_2023/MaterialFunctions/MFC_AnimeShader_2023_Pre}
  }
  MaterialExpressionMax_0 {
   class=MaterialExpressionMax
   inputs [
    A {from=MaterialExpressionMultiply_9}
    B {from=MaterialExpressionMultiply_3}
   ]
  }
  MaterialExpressionMax_1 {
   class=MaterialExpressionMax
   inputs [
    A {from=MaterialExpressionScalarParameter_8}
    B {from=MaterialExpressionScalarParameter_14}
   ]
  }
  MaterialExpressionMax_2 {
   class=MaterialExpressionMax
   inputs [
    A {from=MaterialExpressionConvert_1}
    B {from=MaterialExpressionConvert_1}
   ]
  }
  MaterialExpressionMax_3 {
   class=MaterialExpressionMax
   inputs [
    A {from=MaterialExpressionNamedRerouteUsage_3}
    B {from=MaterialExpressionMax_2}
   ]
  }
  MaterialExpressionMultiply_0 {
   class=MaterialExpressionMultiply
   inputs [
    A {from=MaterialExpressionLinearInterpolate_2}
    B {from=MaterialExpressionScalarParameter_0}
   ]
  }
  MaterialExpressionMultiply_1 {
   class=MaterialExpressionMultiply
   inputs [
    A {from=MaterialExpressionNamedRerouteUsage_4}
    B {from=MaterialExpressionNamedRerouteUsage_1}
   ]
  }
  MaterialExpressionMultiply_10 {
   class=MaterialExpressionMultiply
   inputs [
    A {from=MaterialExpressionMultiply_5}
    B {from=MaterialExpressionMultiply_4}
   ]
  }
  MaterialExpressionMultiply_11 {
   class=MaterialExpressionMultiply
   inputs [
    A {from=MaterialExpressionMultiply_13}
    B {from=""}
   ]
   properties {constB=0.9}
  }
  MaterialExpressionMultiply_12 {
   class=MaterialExpressionMultiply
   inputs [
    A {from=MaterialExpressionVectorParameter_0, output=RGB}
    B {from=""}
   ]
   properties {constB=0.9}
  }
  MaterialExpressionMultiply_13 {
   class=MaterialExpressionMultiply
   inputs [
    A {from=MaterialExpressionNamedRerouteUsage_0}
    B {from=MaterialExpressionScalarParameter_17}
   ]
  }
  MaterialExpressionMultiply_14 {
   class=MaterialExpressionMultiply
   inputs [
    A {from=MaterialExpressionMultiply_15}
    B {from=MaterialExpressionObjectPositionWS_0}
   ]
  }
  MaterialExpressionMultiply_15 {
   class=MaterialExpressionMultiply
   inputs [
    A {from=MaterialExpressionIf_0}
    B {from=MaterialExpressionOneMinus_0}
   ]
  }
  MaterialExpressionMultiply_17 {
   class=MaterialExpressionMultiply
   inputs [
    A {from=MaterialExpressionSubtract_3}
    B {from=MaterialExpressionScalarParameter_22}
   ]
  }
  MaterialExpressionMultiply_2 {
   class=MaterialExpressionMultiply
   inputs [
    A {from=MaterialExpressionNamedRerouteUsage_5}
    B {from=MaterialExpressionNamedRerouteUsage_2}
   ]
  }
  MaterialExpressionMultiply_3 {
   class=MaterialExpressionMultiply
   inputs [
    A {from=MaterialExpressionNamedRerouteDeclaration_5}
    B {from=MaterialExpressionNamedRerouteUsage_28}
   ]
   properties {constB=0}
  }
  MaterialExpressionMultiply_4 {
   class=MaterialExpressionMultiply
   inputs [
    A {from=MaterialExpressionNamedRerouteUsage_30}
    B {from=MaterialExpressionNamedRerouteUsage_7}
   ]
  }
  MaterialExpressionMultiply_5 {
   class=MaterialExpressionMultiply
   inputs [
    A {from=MaterialExpressionPower_4}
    B {from=""}
   ]
   properties {constB=1.5}
  }
  MaterialExpressionMultiply_6 {
   class=MaterialExpressionMultiply
   inputs [
    A {from=MaterialExpressionNamedRerouteUsage_10}
    B {from=MaterialExpressionNamedRerouteUsage_35}
   ]
  }
  MaterialExpressionMultiply_7 {
   class=MaterialExpressionMultiply
   inputs [
    A {from=MaterialExpressionTextureSampleParameter2D_3, output=A}
    B {from=""}
   ]
   properties {constB=2}
  }
  MaterialExpressionMultiply_8 {
   class=MaterialExpressionMultiply
   inputs [
    A {from=MaterialExpressionNamedRerouteUsage_55}
    B {from=MaterialExpressionLinearInterpolate_30}
   ]
  }
  MaterialExpressionMultiply_9 {
   class=MaterialExpressionMultiply
   inputs [
    A {from=MaterialExpressionNamedRerouteDeclaration_10}
    B {from=MaterialExpressionNamedRerouteUsage_6}
   ]
  }
  MaterialExpressionNamedRerouteDeclaration_0 {
   class=MaterialExpressionNamedRerouteDeclaration
   inputs [None {from=MaterialExpressionLinearInterpolate_4}]
   properties {name=ShadowMask, nodeColor=<rgba 0 1 0.7891 1>}
  }
  MaterialExpressionNamedRerouteDeclaration_1 {
   class=MaterialExpressionNamedRerouteDeclaration
   inputs [None {from=MaterialExpressionMultiply_17}]
   properties {name=Ozone Head, nodeColor=<rgba 0 1 0.1094 1>}
  }
  MaterialExpressionNamedRerouteDeclaration_10 {
   class=MaterialExpressionNamedRerouteDeclaration
   inputs [None {from=MaterialExpressionComponentMask_0}]
   properties {name=RightEye, nodeColor=<rgba 1 0 0.0234 1>}
  }
  MaterialExpressionNamedRerouteDeclaration_11 {
   class=MaterialExpressionNamedRerouteDeclaration
   inputs [None {from=MaterialExpressionSwitch_0}]
   properties {name=R_Eye_index_Mask, nodeColor=<rgba 0 0.8125 1 1>}
  }
  MaterialExpressionNamedRerouteDeclaration_12 {
   class=MaterialExpressionNamedRerouteDeclaration
   inputs [None {from=MaterialExpressionLinearInterpolate_13}]
   properties {name=FaceMask, nodeColor=<rgba 1 0 0.3047 1>}
  }
  MaterialExpressionNamedRerouteDeclaration_13 {
   class=MaterialExpressionNamedRerouteDeclaration
   inputs [None {from=MaterialExpressionMaterialFunctionCall_21, output=Color_Lit}]
   properties {name=Input_Color_Lit, nodeColor=<rgba 0 0.625 1 1>}
  }
  MaterialExpressionNamedRerouteDeclaration_14 {
   class=MaterialExpressionNamedRerouteDeclaration
   inputs [None {from=MaterialExpressionMaterialFunctionCall_21, output=Color_Shaded}]
   properties {name=Input_Color_Shaded, nodeColor=<rgba 0.9453 1 0 1>}
  }
  MaterialExpressionNamedRerouteDeclaration_15 {
   class=MaterialExpressionNamedRerouteDeclaration
   inputs [None {from=MaterialExpressionSaturate_3}]
   properties {name=EyeMask, nodeColor=<rgba 1 0.4922 0 1>}
  }
  MaterialExpressionNamedRerouteDeclaration_16 {
   class=MaterialExpressionNamedRerouteDeclaration
   inputs [None {from=MaterialExpressionMaterialFunctionCall_21, output=DFL_Color}]
   properties {name=Input_DFL_Color, nodeColor=<rgba 0.1719 0 1 1>}
  }
  MaterialExpressionNamedRerouteDeclaration_17 {
   class=MaterialExpressionNamedRerouteDeclaration
   inputs [None {from=MaterialExpressionMaterialFunctionCall_1, output=Material}]
   properties {name=Flipbook_MA, nodeColor=<rgba 0.1484 1 0 1>}
  }
  MaterialExpressionNamedRerouteDeclaration_18 {
   class=MaterialExpressionNamedRerouteDeclaration
   inputs [None {from=MaterialExpressionMultiply_0}]
   properties {name=Input_FaceColor, nodeColor=<rgba 1 0 0.4687 1>}
  }
  MaterialExpressionNamedRerouteDeclaration_19 {
   class=MaterialExpressionNamedRerouteDeclaration
   inputs [None {from=MaterialExpressionSwitch_1}]
   properties {name=L_Eye_index_Mask, nodeColor=<rgba 0 0.8125 1 1>}
  }
  MaterialExpressionNamedRerouteDeclaration_2 {
   class=MaterialExpressionNamedRerouteDeclaration
   inputs [None {from=MaterialExpressionPower_0}]
   properties {name=FaceAlpha, nodeColor=<rgba 0 0.5078 1 1>}
  }
  MaterialExpressionNamedRerouteDeclaration_20 {
   class=MaterialExpressionNamedRerouteDeclaration
   inputs [None {from=MaterialExpressionMultiply_8}]
   properties {name=EyeblinkAnim, nodeColor=<rgba 0.7344 0 1 1>}
  }
  MaterialExpressionNamedRerouteDeclaration_21 {
   class=MaterialExpressionNamedRerouteDeclaration
   inputs [None {from=MaterialExpressionPower_3}]
   properties {name=EyeBlinkMask, nodeColor=<rgba 0.3359 0 1 1>}
  }
  MaterialExpressionNamedRerouteDeclaration_22 {
   class=MaterialExpressionNamedRerouteDeclaration
   inputs [None {from=MaterialExpressionAdd_2}]
   properties {name=EyeUVScale, nodeColor=<rgba 0.2656 0 1 1>}
  }
  MaterialExpressionNamedRerouteDeclaration_23 {
   class=MaterialExpressionNamedRerouteDeclaration
   inputs [None {from=MaterialExpressionTextureSampleParameter2D_3, output=B}]
   properties {name=NoseMask, nodeColor=<rgba 0.7812 1 0 1>}
  }
  MaterialExpressionNamedRerouteDeclaration_24 {
   class=MaterialExpressionNamedRerouteDeclaration
   inputs [None {from=MaterialExpressionLinearInterpolate_24}]
   properties {name=EyeThinShadowAnim, nodeColor=<rgba 0 1 0.4141 1>}
  }
  MaterialExpressionNamedRerouteDeclaration_3 {
   class=MaterialExpressionNamedRerouteDeclaration
   inputs [None {from=MaterialExpressionLinearInterpolate_28}]
   properties {name=output_ColorLit, nodeColor=<rgba 1 0.2109 0 1>}
  }
  MaterialExpressionNamedRerouteDeclaration_4 {
   class=MaterialExpressionNamedRerouteDeclaration
   inputs [None {from=MaterialExpressionLinearInterpolate_29}]
   properties {name=output_ColorShaded, nodeColor=<rgba 1 0.2109 0 1>}
  }
  MaterialExpressionNamedRerouteDeclaration_5 {
   class=MaterialExpressionNamedRerouteDeclaration
   inputs [None {from=MaterialExpressionComponentMask_3}]
   properties {name=LeftEye, nodeColor=<rgba 1 0 0.0234 1>}
  }
  MaterialExpressionNamedRerouteDeclaration_6 {
   class=MaterialExpressionNamedRerouteDeclaration
   inputs [None {from=MaterialExpressionSaturate_1}]
   properties {name=EyeThinShadow, nodeColor=<rgba 1 0.375 0 1>}
  }
  MaterialExpressionNamedRerouteDeclaration_7 {
   class=MaterialExpressionNamedRerouteDeclaration
   inputs [None {from=MaterialExpressionTextureSampleParameter2D_3, output=RGB}]
   properties {name=EyeMaskRGB, nodeColor=<rgba 1 0 0.1641 1>}
  }
  MaterialExpressionNamedRerouteDeclaration_8 {
   class=MaterialExpressionNamedRerouteDeclaration
   inputs [None {from=MaterialExpressionSaturate_5}]
   properties {name=Fix_ middle_seam, nodeColor=<rgba 0.6875 0 1 1>}
  }
  MaterialExpressionNamedRerouteDeclaration_9 {
   class=MaterialExpressionNamedRerouteDeclaration
   inputs [None {from=MaterialExpressionSaturate_6}]
   properties {name=Nose, nodeColor=<rgba 1 0.8438 0 1>}
  }
  MaterialExpressionNamedRerouteUsage_0 {class=MaterialExpressionNamedRerouteUsage}
  MaterialExpressionNamedRerouteUsage_1 {class=MaterialExpressionNamedRerouteUsage}
  MaterialExpressionNamedRerouteUsage_10 {class=MaterialExpressionNamedRerouteUsage}
  MaterialExpressionNamedRerouteUsage_11 {class=MaterialExpressionNamedRerouteUsage}
  MaterialExpressionNamedRerouteUsage_13 {class=MaterialExpressionNamedRerouteUsage}
  MaterialExpressionNamedRerouteUsage_14 {class=MaterialExpressionNamedRerouteUsage}
  MaterialExpressionNamedRerouteUsage_15 {class=MaterialExpressionNamedRerouteUsage}
  MaterialExpressionNamedRerouteUsage_16 {class=MaterialExpressionNamedRerouteUsage}
  MaterialExpressionNamedRerouteUsage_17 {class=MaterialExpressionNamedRerouteUsage}
  MaterialExpressionNamedRerouteUsage_18 {class=MaterialExpressionNamedRerouteUsage}
  MaterialExpressionNamedRerouteUsage_19 {class=MaterialExpressionNamedRerouteUsage}
  MaterialExpressionNamedRerouteUsage_2 {class=MaterialExpressionNamedRerouteUsage}
  MaterialExpressionNamedRerouteUsage_20 {class=MaterialExpressionNamedRerouteUsage}
  MaterialExpressionNamedRerouteUsage_21 {class=MaterialExpressionNamedRerouteUsage}
  MaterialExpressionNamedRerouteUsage_22 {class=MaterialExpressionNamedRerouteUsage}
  MaterialExpressionNamedRerouteUsage_23 {class=MaterialExpressionNamedRerouteUsage}
  MaterialExpressionNamedRerouteUsage_24 {class=MaterialExpressionNamedRerouteUsage}
  MaterialExpressionNamedRerouteUsage_25 {class=MaterialExpressionNamedRerouteUsage}
  MaterialExpressionNamedRerouteUsage_26 {class=MaterialExpressionNamedRerouteUsage}
  MaterialExpressionNamedRerouteUsage_27 {class=MaterialExpressionNamedRerouteUsage}
  MaterialExpressionNamedRerouteUsage_28 {class=MaterialExpressionNamedRerouteUsage}
  MaterialExpressionNamedRerouteUsage_3 {class=MaterialExpressionNamedRerouteUsage}
  MaterialExpressionNamedRerouteUsage_30 {class=MaterialExpressionNamedRerouteUsage}
  MaterialExpressionNamedRerouteUsage_31 {class=MaterialExpressionNamedRerouteUsage}
  MaterialExpressionNamedRerouteUsage_32 {class=MaterialExpressionNamedRerouteUsage}
  MaterialExpressionNamedRerouteUsage_33 {class=MaterialExpressionNamedRerouteUsage}
  MaterialExpressionNamedRerouteUsage_34 {class=MaterialExpressionNamedRerouteUsage}
  MaterialExpressionNamedRerouteUsage_35 {class=MaterialExpressionNamedRerouteUsage}
  MaterialExpressionNamedRerouteUsage_36 {class=MaterialExpressionNamedRerouteUsage}
  MaterialExpressionNamedRerouteUsage_38 {class=MaterialExpressionNamedRerouteUsage}
  MaterialExpressionNamedRerouteUsage_39 {class=MaterialExpressionNamedRerouteUsage}
  MaterialExpressionNamedRerouteUsage_4 {class=MaterialExpressionNamedRerouteUsage}
  MaterialExpressionNamedRerouteUsage_40 {class=MaterialExpressionNamedRerouteUsage}
  MaterialExpressionNamedRerouteUsage_41 {class=MaterialExpressionNamedRerouteUsage}
  MaterialExpressionNamedRerouteUsage_42 {class=MaterialExpressionNamedRerouteUsage}
  MaterialExpressionNamedRerouteUsage_43 {class=MaterialExpressionNamedRerouteUsage}
  MaterialExpressionNamedRerouteUsage_46 {class=MaterialExpressionNamedRerouteUsage}
  MaterialExpressionNamedRerouteUsage_47 {class=MaterialExpressionNamedRerouteUsage}
  MaterialExpressionNamedRerouteUsage_48 {class=MaterialExpressionNamedRerouteUsage}
  MaterialExpressionNamedRerouteUsage_49 {class=MaterialExpressionNamedRerouteUsage}
  MaterialExpressionNamedRerouteUsage_5 {class=MaterialExpressionNamedRerouteUsage}
  MaterialExpressionNamedRerouteUsage_51 {class=MaterialExpressionNamedRerouteUsage}
  MaterialExpressionNamedRerouteUsage_55 {class=MaterialExpressionNamedRerouteUsage}
  MaterialExpressionNamedRerouteUsage_56 {class=MaterialExpressionNamedRerouteUsage}
  MaterialExpressionNamedRerouteUsage_57 {class=MaterialExpressionNamedRerouteUsage}
  MaterialExpressionNamedRerouteUsage_6 {class=MaterialExpressionNamedRerouteUsage}
  MaterialExpressionNamedRerouteUsage_7 {class=MaterialExpressionNamedRerouteUsage}
  MaterialExpressionNamedRerouteUsage_8 {class=MaterialExpressionNamedRerouteUsage}
  MaterialExpressionNamedRerouteUsage_9 {class=MaterialExpressionNamedRerouteUsage}
  MaterialExpressionObjectPositionWS_0 {class=MaterialExpressionObjectPositionWS, inputs [], properties {}}
  MaterialExpressionOneMinus_0 {class=MaterialExpressionOneMinus, inputs [None {from=MaterialExpressionVertexColor_0, output=G}]}
  MaterialExpressionPower_0 {
   class=MaterialExpressionPower
   inputs [
    Base {from=MaterialExpressionMaterialFunctionCall_1, output=FaceAlpha}
    Exp {from=MaterialExpressionScalarParameter_3}
   ]
   properties {constExponent=1}
  }
  MaterialExpressionPower_1 {
   class=MaterialExpressionPower
   inputs [
    Base {from=MaterialExpressionMax_0}
    Exp {from=""}
   ]
  }
  MaterialExpressionPower_2 {
   class=MaterialExpressionPower
   inputs [
    Base {from=MaterialExpressionMultiply_7}
    Exp {from=""}
   ]
  }
  MaterialExpressionPower_3 {
   class=MaterialExpressionPower
   inputs [
    Base {from=MaterialExpressionTextureSample_0, output=R}
    Exp {from=""}
   ]
  }
  MaterialExpressionPower_4 {
   class=MaterialExpressionPower
   inputs [
    Base {from=MaterialExpressionNamedRerouteUsage_57}
    Exp {from=""}
   ]
  }
  MaterialExpressionReroute_0 {class=MaterialExpressionReroute, inputs [None {from=MaterialExpressionLinearInterpolate_12}]}
  MaterialExpressionReroute_2 {class=MaterialExpressionReroute, inputs [None {from=MaterialExpressionMaterialFunctionCall_21, output=Emissive}]}
  MaterialExpressionReroute_3 {class=MaterialExpressionReroute, inputs [None {from=MaterialExpressionSaturate_0}]}
  MaterialExpressionSaturate_0 {class=MaterialExpressionSaturate, inputs [None {from=MaterialExpressionNamedRerouteUsage_24}]}
  MaterialExpressionSaturate_1 {class=MaterialExpressionSaturate, inputs [None {from=MaterialExpressionPower_2}]}
  MaterialExpressionSaturate_3 {class=MaterialExpressionSaturate, inputs [None {from=MaterialExpressionPower_1}]}
  MaterialExpressionSaturate_5 {class=MaterialExpressionSaturate, inputs [None {from=MaterialExpressionPower_1}]}
  MaterialExpressionSaturate_6 {class=MaterialExpressionSaturate, inputs [None {from=MaterialExpressionMultiply_10}]}
  MaterialExpressionScalarParameter_0 {class=MaterialExpressionScalarParameter, properties {defaultValue=1, parameterName=FlipbookFaceAnimeShadedValue}}
  MaterialExpressionScalarParameter_1 {class=MaterialExpressionScalarParameter, properties {defaultValue=0.65, group=Flipbook_shadow, parameterName=ShadowIntensity}}
  MaterialExpressionScalarParameter_10 {class=MaterialExpressionScalarParameter, properties {defaultValue=1, group=Flipbook 03 Eyes, parameterName=FB_EyeUVScaleX, sortPriority=13}}
  MaterialExpressionScalarParameter_11 {class=MaterialExpressionScalarParameter, properties {group=Flipbook 03 Eyes, parameterName=FB_EyeUVOffsetX, sortPriority=16}}
  MaterialExpressionScalarParameter_12 {class=MaterialExpressionScalarParameter, properties {group=Flipbook 03 Eyes, parameterName=FB_EyeUVOffsetY, sortPriority=17}}
  MaterialExpressionScalarParameter_13 {class=MaterialExpressionScalarParameter, properties {group=Temporal Responsiveness, parameterName=HasPoseChanged}}
  MaterialExpressionScalarParameter_14 {class=MaterialExpressionScalarParameter, properties {group=Temporal Responsiveness, parameterName=HasLookDirectionChanged}}
  MaterialExpressionScalarParameter_15 {class=MaterialExpressionScalarParameter, properties {defaultValue=1, group=Flipbook 03 Eyes, parameterName=FB_EyeUVScaleY, sortPriority=14}}
  MaterialExpressionScalarParameter_16 {class=MaterialExpressionScalarParameter, properties {group=Temporal Responsiveness, parameterName=DebugTemporalResponsiveness}}
  MaterialExpressionScalarParameter_17 {class=MaterialExpressionScalarParameter, properties {defaultValue=0.92, parameterName=Eye_BlinkColor_brightness}}
  MaterialExpressionScalarParameter_18 {
   class=MaterialExpressionScalarParameter
   properties {
    desc="Driven by Animation Blueprint.  Do not set manually."
    group=Flipbook 05 Indices
    parameterName=flipbook_face_L_eye_index
    sortPriority=10
   }
  }
  MaterialExpressionScalarParameter_19 {
   class=MaterialExpressionScalarParameter
   properties {desc="Driven by Ation Blueprint.  Do not set manually.", group=Flipbook 05 Indices, parameterName=flipbook_face_R_eye_index, sortPriority=11}
  }
  MaterialExpressionScalarParameter_2 {class=MaterialExpressionScalarParameter, properties {group=Temporal Responsiveness, parameterName=HasLookDirectionChanged}}
  MaterialExpressionScalarParameter_20 {
   class=MaterialExpressionScalarParameter
   properties {
    desc="Driven by Animation Blueprint.  Do not set manually."
    group=Flipbook 05 Indices
    parameterName=flipbook_face_L_eye_index
    sortPriority=10
   }
  }
  MaterialExpressionScalarParameter_21 {
   class=MaterialExpressionScalarParameter
   properties {desc="Driven by Ation Blueprint.  Do not set manually.", group=Flipbook 05 Indices, parameterName=flipbook_face_R_eye_index, sortPriority=11}
  }
  MaterialExpressionScalarParameter_22 {class=MaterialExpressionScalarParameter, properties {parameterName=UseOzoneHead}}
  MaterialExpressionScalarParameter_25 {class=MaterialExpressionScalarParameter, properties {defaultValue=1, group=Flipbook 03 Eyes, parameterName=FB_EyeUVScale, sortPriority=12}}
  MaterialExpressionScalarParameter_26 {
   class=MaterialExpressionScalarParameter
   properties {
    desc="Driven by Animation Blueprint.  Do not set manually."
    group=Flipbook 05 Indices
    parameterName=flipbook_face_L_eye_index
    sortPriority=10
   }
  }
  MaterialExpressionScalarParameter_27 {
   class=MaterialExpressionScalarParameter
   properties {desc="Driven by Ation Blueprint.  Do not set manually.", group=Flipbook 05 Indices, parameterName=flipbook_face_R_eye_index, sortPriority=11}
  }
  MaterialExpressionScalarParameter_28 {
   class=MaterialExpressionScalarParameter
   properties {
    desc="Driven by Animation Blueprint.  Do not set manually."
    group=Flipbook 05 Indices
    parameterName=flipbook_face_L_eye_index
    sortPriority=10
   }
  }
  MaterialExpressionScalarParameter_29 {
   class=MaterialExpressionScalarParameter
   properties {desc="Driven by Ation Blueprint.  Do not set manually.", group=Flipbook 05 Indices, parameterName=flipbook_face_R_eye_index, sortPriority=11}
  }
  MaterialExpressionScalarParameter_3 {class=MaterialExpressionScalarParameter, properties {defaultValue=1, group=Flipbppk_shadow, parameterName=Alpha contrast}}
  MaterialExpressionScalarParameter_30 {
   class=MaterialExpressionScalarParameter
   properties {desc="Driven by Ation Blueprint.  Do not set manually.", group=Flipbook 05 Indices, parameterName=flipbook_face_R_eye_index, sortPriority=11}
  }
  MaterialExpressionScalarParameter_32 {
   class=MaterialExpressionScalarParameter
   properties {
    desc="Driven by Animation Blueprint.  Do not set manually."
    group=Flipbook 05 Indices
    parameterName=flipbook_face_L_eye_index
    sortPriority=10
   }
  }
  MaterialExpressionScalarParameter_4 {
   class=MaterialExpressionScalarParameter
   properties {
    desc="Driven by Animation Blueprint.  Do not set manually."
    group=Flipbook 05 Indices
    parameterName=flipbook_face_L_eye_index
    sortPriority=10
   }
  }
  MaterialExpressionScalarParameter_5 {
   class=MaterialExpressionScalarParameter
   properties {desc="Driven by Ation Blueprint.  Do not set manually.", group=Flipbook 05 Indices, parameterName=flipbook_face_R_eye_index, sortPriority=11}
  }
  MaterialExpressionScalarParameter_6 {class=MaterialExpressionScalarParameter, properties {group=Flipbook 03 Eyes, parameterName=FB_EyeUVScaleX_Fix, sortPriority=13}}
  MaterialExpressionScalarParameter_7 {class=MaterialExpressionScalarParameter, properties {group=Flipbook 03 Eyes, parameterName=FB_EyeUVScaleY_Fix, sortPriority=14}}
  MaterialExpressionScalarParameter_8 {class=MaterialExpressionScalarParameter, properties {group=Temporal Responsiveness, parameterName=HasPoseChanged}}
  MaterialExpressionScalarParameter_9 {class=MaterialExpressionScalarParameter, properties {defaultValue=1, group=01_EyeAnimation_Parameters, parameterName=UseFaceMask}}
  MaterialExpressionSetMaterialAttributes_0 {
   class=MaterialExpressionSetMaterialAttributes
   inputs [
    MaterialAttributes {from=MaterialExpressionGetMaterialAttributes_1, output=MaterialAttributes}
    World Position Offset {from=MaterialExpressionLinearInterpolate_10, output=""}
    Emissive Color {from=MaterialExpressionLinearInterpolate_14, output=""}
   ]
   properties {attributeSetTypes [#1, #2]}
  }
  MaterialExpressionStaticBoolParameter_0 {
   class=MaterialExpressionStaticBoolParameter
   inputs []
   properties {defaultValue=true, group=01_EyeAnimation_Parameters, parameterName=Use UV2 for Eye Animation, sortPriority=-30}
  }
  MaterialExpressionStaticBoolParameter_1 {
   class=MaterialExpressionStaticBoolParameter
   inputs []
   properties {defaultValue=true, group=01_EyeAnimation_Parameters, parameterName=Use UV2 for Eye Animation, sortPriority=-30}
  }
  MaterialExpressionStaticSwitchParameter_4 {
   class=MaterialExpressionStaticSwitchParameter
   inputs [
    True {from=MaterialExpressionAppendVector_13, output=""}
    False {from=MaterialExpressionScalarParameter_25, output=""}
   ]
   properties {group=Flipbook 03 Eyes, parameterName=FB_EyeUVNonUniformScale, sortPriority=11}
  }
  MaterialExpressionStaticSwitch_2 {
   class=MaterialExpressionStaticSwitch
   inputs [
    True {from=MaterialExpressionTextureCoordinate_5, output=""}
    False {from=MaterialExpressionTextureCoordinate_4, output=""}
    Value {from=MaterialExpressionStaticBoolParameter_1, output=""}
   ]
   properties {}
  }
  MaterialExpressionSubtract_0 {
   class=MaterialExpressionSubtract
   inputs [
    A {from=MaterialExpressionNamedRerouteUsage_56}
    B {from=MaterialExpressionNamedRerouteUsage_51}
   ]
  }
  MaterialExpressionSubtract_1 {
   class=MaterialExpressionSubtract
   inputs [
    A {from=MaterialExpressionNamedRerouteUsage_51}
    B {from=MaterialExpressionNamedRerouteUsage_56}
   ]
  }
  MaterialExpressionSubtract_3 {
   class=MaterialExpressionSubtract
   inputs [
    A {from=MaterialExpressionMultiply_15}
    B {from=MaterialExpressionMultiply_14}
   ]
  }
  MaterialExpressionSubtract_4 {
   class=MaterialExpressionSubtract
   inputs [
    A {from=MaterialExpressionNamedRerouteUsage_11}
    B {from=MaterialExpressionAdd_0}
   ]
  }
  MaterialExpressionSwitch_0 {
   class=MaterialExpressionSwitch
   inputs [
    SwitchValue {from=MaterialExpressionScalarParameter_5, output=""}
    Default {from=MaterialExpressionConstant_6, output=""}
    None {from=MaterialExpressionConstant_6, output=""}
    None {from=MaterialExpressionConstant_7, output=""}
    None {from=MaterialExpressionConstant_7, output=""}
    None {from=MaterialExpressionConstant_7, output=""}
    None {from=MaterialExpressionConstant_8, output=""}
    None {from=MaterialExpressionConstant_8, output=""}
    None {from=MaterialExpressionConstant_8, output=""}
    None {from=MaterialExpressionConstant_8, output=""}
    None {from=MaterialExpressionConstant_8, output=""}
   ]
   properties {inputs [
     {
      inputName=None
      input {
       expression=&self:MaterialExpressionConstant_6
       inputName=None
       mask=0
       maskA=0
       maskB=0
       maskG=0
       maskR=0
       outputIndex=0
      }
     }
     {
      inputName=None
      input {
       expression=&self:MaterialExpressionConstant_7
       inputName=None
       mask=0
       maskA=0
       maskB=0
       maskG=0
       maskR=0
       outputIndex=0
      }
     }
     {
      inputName=None
      input {
       expression=&self:MaterialExpressionConstant_7
       inputName=None
       mask=0
       maskA=0
       maskB=0
       maskG=0
       maskR=0
       outputIndex=0
      }
     }
     {
      inputName=None
      input {
       expression=&self:MaterialExpressionConstant_7
       inputName=None
       mask=0
       maskA=0
       maskB=0
       maskG=0
       maskR=0
       outputIndex=0
      }
     }
     {
      inputName=None
      input {
       expression=&self:MaterialExpressionConstant_8
       inputName=None
       mask=0
       maskA=0
       maskB=0
       maskG=0
       maskR=0
       outputIndex=0
      }
     }
     {
      inputName=None
      input {
       expression=&self:MaterialExpressionConstant_8
       inputName=None
       mask=0
       maskA=0
       maskB=0
       maskG=0
       maskR=0
       outputIndex=0
      }
     }
     {
      inputName=None
      input {
       expression=&self:MaterialExpressionConstant_8
       inputName=None
       mask=0
       maskA=0
       maskB=0
       maskG=0
       maskR=0
       outputIndex=0
      }
     }
     {
      inputName=None
      input {
       expression=&self:MaterialExpressionConstant_8
       inputName=None
       mask=0
       maskA=0
       maskB=0
       maskG=0
       maskR=0
       outputIndex=0
      }
     }
     {
      inputName=None
      input {
       expression=&self:MaterialExpressionConstant_8
       inputName=None
       mask=0
       maskA=0
       maskB=0
       maskG=0
       maskR=0
       outputIndex=0
      }
     }
    ]}
  }
  MaterialExpressionSwitch_1 {
   class=MaterialExpressionSwitch
   inputs [
    SwitchValue {from=MaterialExpressionScalarParameter_4, output=""}
    Default {from=MaterialExpressionConstant_9, output=""}
    None {from=MaterialExpressionConstant_9, output=""}
    None {from=MaterialExpressionConstant_10, output=""}
    None {from=MaterialExpressionConstant_10, output=""}
    None {from=MaterialExpressionConstant_10, output=""}
    None {from=MaterialExpressionConstant_11, output=""}
    None {from=MaterialExpressionConstant_11, output=""}
    None {from=MaterialExpressionConstant_11, output=""}
    None {from=MaterialExpressionConstant_11, output=""}
    None {from=MaterialExpressionConstant_11, output=""}
   ]
   properties {inputs [
     {
      inputName=None
      input {
       expression=&self:MaterialExpressionConstant_9
       inputName=None
       mask=0
       maskA=0
       maskB=0
       maskG=0
       maskR=0
       outputIndex=0
      }
     }
     {
      inputName=None
      input {
       expression=&self:MaterialExpressionConstant_10
       inputName=None
       mask=0
       maskA=0
       maskB=0
       maskG=0
       maskR=0
       outputIndex=0
      }
     }
     {
      inputName=None
      input {
       expression=&self:MaterialExpressionConstant_10
       inputName=None
       mask=0
       maskA=0
       maskB=0
       maskG=0
       maskR=0
       outputIndex=0
      }
     }
     {
      inputName=None
      input {
       expression=&self:MaterialExpressionConstant_10
       inputName=None
       mask=0
       maskA=0
       maskB=0
       maskG=0
       maskR=0
       outputIndex=0
      }
     }
     {
      inputName=None
      input {
       expression=&self:MaterialExpressionConstant_11
       inputName=None
       mask=0
       maskA=0
       maskB=0
       maskG=0
       maskR=0
       outputIndex=0
      }
     }
     {
      inputName=None
      input {
       expression=&self:MaterialExpressionConstant_11
       inputName=None
       mask=0
       maskA=0
       maskB=0
       maskG=0
       maskR=0
       outputIndex=0
      }
     }
     {
      inputName=None
      input {
       expression=&self:MaterialExpressionConstant_11
       inputName=None
       mask=0
       maskA=0
       maskB=0
       maskG=0
       maskR=0
       outputIndex=0
      }
     }
     {
      inputName=None
      input {
       expression=&self:MaterialExpressionConstant_11
       inputName=None
       mask=0
       maskA=0
       maskB=0
       maskG=0
       maskR=0
       outputIndex=0
      }
     }
     {
      inputName=None
      input {
       expression=&self:MaterialExpressionConstant_11
       inputName=None
       mask=0
       maskA=0
       maskB=0
       maskG=0
       maskR=0
       outputIndex=0
      }
     }
    ]}
  }
  MaterialExpressionTemporalResponsivenessOutput_1 {class=MaterialExpressionTemporalResponsivenessOutput, inputs [None {from=MaterialExpressionLinearInterpolate_3, output=""}], properties {}}
  MaterialExpressionTextureCoordinate_4 {class=MaterialExpressionTextureCoordinate}
  MaterialExpressionTextureCoordinate_5 {class=MaterialExpressionTextureCoordinate, properties {coordinateIndex=1}}
  MaterialExpressionTextureSampleParameter2D_1 {
   class=MaterialExpressionTextureSampleParameter2D
   inputs [
    UVs {from="", output=""}
    Apply View MipBias {from="", output=""}
   ]
   properties {
    parameterName=ShadowMask
    samplerSource=SSM_Clamp_WorldGroupSettings
    samplerType=SAMPLERTYPE_LinearColor
    texture=/Game/Global/Textures/Default/Blanks/T_BlackLinearColor
   }
  }
  MaterialExpressionTextureSampleParameter2D_3 {
   class=MaterialExpressionTextureSampleParameter2D
   inputs [
    UVs {from=MaterialExpressionMaterialFunctionCall_0, output=UVs}
    Apply View MipBias {from="", output=""}
   ]
   properties {
    group=01_EyeAnimation_Parameters
    parameterName=L_R_EyeMask
    samplerSource=SSM_Clamp_WorldGroupSettings
    samplerType=SAMPLERTYPE_Masks
    texture=/BRCosmetics/Characters/Player/Male/Large/Bodies/M_LRG_PacificSweater/Textures/T_LemurClam_EyeMask_Fixseam
   }
  }
  MaterialExpressionTextureSample_0 {
   class=MaterialExpressionTextureSample
   inputs [
    UVs {from=MaterialExpressionMaterialFunctionCall_0, output=UVs}
    Tex {from=""}
    Apply View MipBias {from=""}
   ]
   properties {
    samplerSource=SSM_Clamp_WorldGroupSettings
    samplerType=SAMPLERTYPE_Masks
    texture=/BRCosmetics/Characters/Player/Male/Large/Bodies/M_LRG_LemurClam/Textures/T_LemurClam_Eyeblink
   }
  }
  MaterialExpressionVectorParameter_0 {
   class=MaterialExpressionVectorParameter
   properties {group=Flipbook 03 Eyes, parameterName=EyeThinShaow Color, defaultValue=<rgba 0.2664 0.1812 0.1046 1>}
  }
  MaterialExpressionVectorParameter_1 {
   class=MaterialExpressionVectorParameter
   properties {group=Flipbook 03 Eyes, parameterName=Eye_BlinkColor, defaultValue=<rgba 0.9 0.631 0.4159 1>}
  }
  MaterialExpressionVectorParameter_2 {
   class=MaterialExpressionVectorParameter
   properties {group=Flipbook 03 Eyes, parameterName=EyeBlinkline Color, defaultValue=<rgba 0.1 0.0713 0.0426 1>}
  }
  MaterialExpressionVertexColor_0 {class=MaterialExpressionVertexColor, inputs [], properties {}}
 }
 properties {
  bAutomaticallySetUsageInEditor=false
  bHasPixelAnimation=true
  bUseMaterialAttributes=true
  bUsedWithMeshDeformer=true
  bUsedWithMorphTargets=true
  bUsedWithSkeletalMesh=true
  pixelDepthOffsetMode=PDOM_Legacy
  shadingModel=MSM_Cloth
  usedShadingModels=Cloth
 }
}
