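# Physics asset for the large male Pacific Sweater body cosmetic, covering its left and right tassels.
# Each side is a mirrored chain of capsules running from a larger handlerBar body through a small base body to three dyn_tassle segments.
# The base joint swings and twists freely on the handlerBar, while the segment joints below carry moderate swing limits and near-zero twist limits.
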PhysicsAsset /BRCosmetics/Characters/Player/Male/Large/Bodies/M_LRG_PacificSweater/Meshes/M_LRG_PacificSweater_Physics
bodies {
 base_tassle_01_l [{shapeType=Capsule, length=0.505, radius=0.505, shapeName=base_tassle_01_l_capsule, rotation=<rot 0 -90 0>}]
 base_tassle_01_r [{shapeType=Capsule, length=0.505, radius=0.505, shapeName=base_tassle_01_r_capsule, rotation=<rot 0 -90 0>}]
 dyn_tassle_01_l [{shapeType=Capsule, length=2.0713, radius=0.9287, shapeName=dyn_tassle_01_l_capsule, center=<1.726 -0.2554 0>, rotation=<rot 90 -21.0527 0>}]
 dyn_tassle_01_r [{shapeType=Capsule, length=2.0713, radius=0.9287, shapeName=dyn_tassle_01_l_capsule, center=<1.726 0.2554 0>, rotation=<rot 90 21.0527 0>}]
 dyn_tassle_02_l [{shapeType=Capsule, length=2.1196, radius=0.8804, shapeName=dyn_tassle_02_l_capsule, center=<1.737 0 0>, rotation=<rot 90 0 0>}]
 dyn_tassle_02_r [{shapeType=Capsule, length=2.1196, radius=0.8804, shapeName=dyn_tassle_02_l_capsule, center=<1.737 0 0>, rotation=<rot 90 0 0>}]
 dyn_tassle_03_l [{
   shapeType=Capsule
   length=3.2834
   radius=1.4492
   shapeName=dyn_tassle_03_l_capsule
   center=<2.8952 0.5528 -0.0437>
   rotation=<rot 90 9.7018 0>
  }]
 dyn_tassle_03_r [{
   shapeType=Capsule
   length=3.2834
   radius=1.4492
   shapeName=dyn_tassle_03_l_capsule
   center=<2.8952 -0.5528 -0.0437>
   rotation=<rot 90 -9.7018 0>
  }]
 handlerBar_l [{
   shapeType=Capsule
   length=8.1197
   radius=4.1465
   shapeName=handlerBar_l_capsule
   center=<8.1023 -5.2978 -3.3104>
   rotation=<rot 0 -108.3298 -80.0167>
  }]
 handlerBar_r [{
   shapeType=Capsule
   length=8.1197
   radius=4.1465
   shapeName=handlerBar_r_capsule
   center=<8.1022 -5.2978 3.3104>
   rotation=<rot 0 71.6702 -80.0167>
  }]
}
constraints [
 {
  bone1Name=base_tassle_01_l
  bone2Name=handlerBar_l
  swing1LimitDegrees=45
  swing1Motion=Free
  swing2LimitDegrees=45
  swing2Motion=Free
  twistLimitDegrees=45
  twistMotion=Free
 }
 {bone1Name=dyn_tassle_01_l, bone2Name=base_tassle_01_l, swing1LimitDegrees=25, swing2LimitDegrees=25, twistLimitDegrees=1}
 {bone1Name=dyn_tassle_02_l, bone2Name=dyn_tassle_01_l, swing1LimitDegrees=25, swing2LimitDegrees=25, twistLimitDegrees=1}
 {bone1Name=dyn_tassle_03_l, bone2Name=dyn_tassle_02_l, swing1LimitDegrees=25, swing2LimitDegrees=25, twistLimitDegrees=1}
 {
  bone1Name=base_tassle_01_r
  bone2Name=handlerBar_r
  swing1LimitDegrees=45
  swing1Motion=Free
  swing2LimitDegrees=45
  swing2Motion=Free
  twistLimitDegrees=45
  twistMotion=Free
 }
 {bone1Name=dyn_tassle_01_r, bone2Name=base_tassle_01_r, swing1LimitDegrees=25, swing2LimitDegrees=25, twistLimitDegrees=1}
 {bone1Name=dyn_tassle_02_r, bone2Name=dyn_tassle_01_r, swing1LimitDegrees=25, swing2LimitDegrees=25, twistLimitDegrees=1}
 {bone1Name=dyn_tassle_03_r, bone2Name=dyn_tassle_02_r, swing1LimitDegrees=25, swing2LimitDegrees=25, twistLimitDegrees=1}
]
