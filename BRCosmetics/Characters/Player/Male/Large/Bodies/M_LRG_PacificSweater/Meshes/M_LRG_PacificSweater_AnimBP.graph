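# Post-copy animation Blueprint for the PacificSweater large male cosmetic body, animating its secondary parts over the parent mesh's pose.
# The copied pose feeds a Control Rig, a spine-relative RigidBody simulation, then additive bone offsets for shoulder pad, trashcan and forearm scale.
# Update events derive rig inputs from movement, head bone deltas, skydive state and emote montage activity.
# Jump events multiply forward movement into the right shoulder pad's rotation alpha.
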AnimBlueprint /BRCosmetics/Characters/Player/Male/Large/Bodies/M_LRG_PacificSweater/Meshes/M_LRG_PacificSweater_AnimBP
target_skeleton=&/M_LRG_PacificSweater_Skeleton
graphs [
 AnimGraph {nodes [
   n1 = Control Rig!(
    Source=Copy Pose From Mesh() {bUseAttachedParent=true}
    Alpha
    "is Targeting ":bool=Get bIsTargeting()
    is Parachuting:bool=Get bIsParachuteOpened()
    is Emoting:bool=Get bIsPlayingEmote()
    Char Forward:double=Get Char Forward() -> (Char Forward:double)
    Char Backwards:double=Get Char Backward() -> (Char Backward:double)
    Char Left:double=Get Char Move Left() -> (Char Move Left:double)
    Char Right:double=0
    Skydive Floating:bool=Get Is Skydive Floating CR() -> (Is Skydive Floating CR:bool)
    Head Down:double=Get Head Down() -> (Head Down:double)
    Head Right:double="Get Head Right "() -> ("Head Right ":double)
    Head Left:double=Get Head Left() -> (Head Left:double)
    pacificsweater_emote_playing:bool=Get pacificsweater_emote_playing() -> (pacificsweater_emote_playing:bool)
   ) -> (Pose) {controlRigAssetReference {blueprintRigClass=&/M_LRG_PacificSweater_CtrlRig.M_LRG_PacificSweater_CtrlRig_C}}
   n2 = RigidBody!(ComponentPose=Local To Component(LocalPose=@n1), Alpha, OverrideWorldGravity:Vector="0,0,-980", ExternalForce) -> (Pose) {
    bClampLinearTranslationLimitToRefPose=true
    bOverrideWorldGravity=true
    overridePhysicsAsset=&/M_LRG_PacificSweater_Physics
    simulationSpace=BaseBoneSpace
    baseBoneRef {boneName=spine_05}
    componentAppliedLinearAccClamp=<0 0 0>
    simSpaceSettings {
     dampingAlpha=0
     maxAngularAcceleration=32
     maxAngularVelocity=8
     maxLinearAcceleration=2800
     maxLinearVelocity=700
     worldAlpha=1
     externalLinearDragV=<0.5 0.5 0.5>
    }
   }
   n3 = "Transform (Modify) Bone - Bone: shoulderPad_Master_r"(
    ComponentPose=@n2
    Alpha=Get Jumping Forward() -> (Jumping Forward:double)
    Translation="0,0,0"
    Rotation="0.000,12.957,0.000"
    Scale="1,1,1"
   ) {
    rotationMode=BMM_Additive
    rotationSpace=BCS_BoneSpace
    translationMode=BMM_Additive
    translationSpace=BCS_BoneSpace
    alphaScaleBiasClamp {bInterpResult=true}
    boneToModify {boneName=shoulderPad_Master_r}
    rotation=<rot 0 12.957 0>
   }
   n4 = "Transform (Modify) Bone - Bone: trashcanBack_Mover"(ComponentPose=@n3, Alpha=1, Translation="0,3.2,0", Rotation="0,0,0", Scale="1,1,1") {
    rotationMode=BMM_Additive
    rotationSpace=BCS_BoneSpace
    translationMode=BMM_Additive
    translationSpace=BCS_BoneSpace
    alphaScaleBiasClamp {bInterpResult=true}
    boneToModify {boneName=trashcanBack_Mover}
    translation {y=3.2}
   }
   n5 = "Transform (Modify) Bone - Bone: lowerarm_r"(ComponentPose=@n4, Alpha=1, Translation="0,0,0", Rotation="0,0,0", Scale="0.3,0.3,0.3") {
    rotationMode=BMM_Additive
    rotationSpace=BCS_BoneSpace
    scaleMode=BMM_Additive
    scaleSpace=BCS_BoneSpace
    translationMode=BMM_Additive
    translationSpace=BCS_BoneSpace
    boneToModify {boneName=lowerarm_r}
    scale=<0.3 0.3 0.3>
   }
   n6 = "Transform (Modify) Bone - Bone: lowerarm_l"(ComponentPose=@n5, Alpha=1, Translation="0,0,0", Rotation="0,0,0", Scale="0.3,0.3,0.3") {
    rotationMode=BMM_Additive
    rotationSpace=BCS_BoneSpace
    scaleMode=BMM_Additive
    scaleSpace=BCS_BoneSpace
    translationMode=BMM_Additive
    translationSpace=BCS_BoneSpace
    boneToModify {boneName=lowerarm_l}
    scale=<0.3 0.3 0.3>
   }
   Output Pose(Result=Component To Local(ComponentPose=@n6)) {layerGroup=None}
   Get Char Move Right() -> (Char Move Right:double)
  ]}
 EventGraph {nodes [
   Event BlueprintUpdateAnimation()
   > n1 = PlayerMovement(AnimInstance=Self-Reference(), XDotProductCoefficient=25, YDotProductCoefficient=25, ZDotProductCoefficient=1)
   > Set Char Move Right(Char Move Right:double=MapRangeClamped(Value=@n1, InRangeA=-1, InRangeB=25)) -> (Output_Get:double)
   > Set Char Move Left(Char Move Left:double=MapRangeClamped(Value=@n1, InRangeA=1, InRangeB=-25)) -> (Output_Get:double)
   > Set Char Forward(Char Forward:double=MapRangeClamped(Value=@n1.YMovement, InRangeA=-1, InRangeB=25)) -> (Output_Get:double)
   > Set Char Backward(Char Backward:double=MapRangeClamped(Value=@n1.YMovement, InRangeA=1, InRangeB=-25)) -> (Output_Get:double)
   > n2 = Set Is Skydive Floating CR(Is Skydive Floating CR:bool=Get bIsSkydiveFloating()) -> (Output_Get:bool)
   n3 = Set SkelMesh(execute=@n7.then, SkelMesh=GetOwnerSkeletalMeshComponent())
   Set Head Down(
    execute=@n2.then
    Head Down:double=GetAlphaFromReferenceTransformDelta(SkeletalMeshComponent=@n3, BoneName=spine_05, BaseName=head, InputRange=(X=0,Y=-20), TwistAxis="0,0,1")
   ) -> (Output_Get:double)
   > Set Head Left(Head Left:double=GetAlphaFromReferenceTransformDelta(SkeletalMeshComponent=@n3, BoneName=spine_05, BaseName=neck_01, DeltaComponentType=EulerX, InputRange=(X=-11,Y=-25), TwistAxis="0,0,1")) -> (Output_Get:double)
   > "Set Head Right "("Head Right ":double=GetAlphaFromReferenceTransformDelta(SkeletalMeshComponent=@n3, BoneName=spine_05, BaseName=neck_01, DeltaComponentType=EulerX, InputRange=(X=-11,Y=5), TwistAxis="0,0,1")) -> (Output_Get:double)
   n4 = GetAnimInstance(self=GetOwnerSkeletalMeshComponent())
   > Is Valid(InputObject=@n4)
   > Set pacificsweater_emote_playing(pacificsweater_emote_playing:bool=Montage_IsActive!(
     self:Anim Instance Object Reference=@n4
     Montage:Anim Montage Object Reference=/BRCosmetics/Animation/Game/MainPlayer/Emotes/PacificSweater/CLM_HERO/Emote_PacificSweater_CLM_HERO_Starting_M
    ) -> (ReturnValue:bool)) -> (Output_Get:bool)
   Event OnBeginJumping()
   > n5 = Set is Jumping(is Jumping:double=1) -> (Output_Get:double)
   Set Jumping Forward(
    execute=[@n6.then, @n5.then]
    Jumping Forward:double=float * float(A=Get is Jumping() -> (is Jumping:double), B=Get Char Forward()) {is=title K2Node_PromotableOperator}
   ) -> (Output_Get:double)
   Event OnEndFalling()
   > n6 = Set is Jumping(is Jumping=0)
   n7 = Event BlueprintInitializeAnimation()
   n8 = Cast To FortPlayerPawn(execute=@n3.then, Object=TryGetFortPawn())
   > Set Pawn(Pawn:Fort Player Pawn Object Reference=@n8) -> (Output_Get:Fort Player Pawn Object Reference)
  ]}
]
variables [
 Char Backward {type=number}
 Char Forward {type=number}
 Char Move Left {type=number}
 Char Move Right {type=number}
 Head Down {type=number}
 Head Left {type=number}
 "Head Right " {type=number}
 is Jumping {type=number}
 Is Skydive Floating CR {type=boolean}
 Jumping Forward {type=number}
 pacificsweater_emote_playing {type=boolean}
 Pawn {type=FortPlayerPawn}
 SkelMesh {type=SkeletalMeshComponent}
]
