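# Control Rig for the PacificSweater body's face-accessory part, animating the small "Kenny" arms and handlebar bones on the character's chest in response to movement and head direction.
# Construction builds head, neck, hand IK and pole-vector controls.
# Forward solve runs the shared DeformRig function, two-bone IK on both arms, then Kenny Arm Movement.
# That last step smooths the Char and Head inputs with AlphaInterp to offset IK controls and rotate handlebars.
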ControlRigBlueprint /BRCosmetics/Characters/Player/Male/Large/Bodies/M_LRG_PacificSweater/Meshes/Parts/M_LRG_PacificSweater_FaceAcc_CtrlRig
graphs [
 New Function_ContainedGraph {nodes [
   RigVMFunctionEntryNode()
   RigUnit_GetTransform = &RigUnit_GetTransform(Item=(Type=Null,Name="ref_PoleVect_L_ctrl_null"), bInitial=False)
   > /Script/ControlRig.RigUnit_TwoBoneIKSimplePerItem(
    DebugSettings=(bEnabled=False,Scale=10,WorldOffset=(Rotation=(X=0,Y=0,Z=0,W=1),Translation=(X=0,Y=0,Z=0),Scale3D=(X=1,Y=1,Z=1)))
    EffectorItem=(Type=Bone,Name="dyn_kennyWrist_l")
    ItemA=(Type=Bone,Name="dyn_kennyUpperarm_l")
    ItemALength=0
    ItemB=(Type=Bone,Name="dyn_kennyLowerarm_l")
    ItemBLength=0
    PoleVectorKind=Direction
    PoleVectorSpace=(Type=Bone,Name="None")
    PrimaryAxis=(X=1,Y=0,Z=0)
    SecondaryAxis=(X=0,Y=-1,Z=0)
    SecondaryAxisWeight=1
    StretchMaximumRatio=1.25
    StretchStartRatio=0.75
    Weight=1
    bEnableStretch=False
    bPropagateToChildren=True
    PoleVector=@RigUnit_GetTransform.Transform.Translation
    Effector=&RigUnit_GetTransform(Item=(Type=Null,Name="ref_HandIK_L_ctrl_null"), bInitial=False).Transform
   ) {is=struct RigVMUnitNode}
   > &RigUnit_TwoBoneIKSimplePerItem(
    DebugSettings=(bEnabled=False,Scale=10,WorldOffset=(Rotation=(X=0,Y=0,Z=0,W=1),Translation=(X=0,Y=0,Z=0),Scale3D=(X=1,Y=1,Z=1)))
    EffectorItem=(Type=Bone,Name="dyn_kennyWrist_r")
    ItemA=(Type=Bone,Name="dyn_kennyUpperarm_r")
    ItemALength=0
    ItemB=(Type=Bone,Name="dyn_kennyLowerarm_r")
    ItemBLength=0
    PoleVectorKind=Direction
    PoleVectorSpace=(Type=Bone,Name="None")
    PrimaryAxis=(X=1,Y=0,Z=0)
    SecondaryAxis=(X=0,Y=1,Z=0)
    SecondaryAxisWeight=1
    StretchMaximumRatio=1.25
    StretchStartRatio=0.75
    Weight=1
    bEnableStretch=False
    bPropagateToChildren=True
    PoleVector=&RigUnit_GetTransform(Item=(Type=Null,Name="ref_PoleVect_R_ctrl_null"), bInitial=False).Transform.Translation
    Effector=&RigUnit_GetTransform(Item=(Type=Null,Name="ref_HandIK_R_ctrl_null"), bInitial=False).Transform
   )
   > RigVMFunctionReturnNode()
  ]}
 New Function_ContainedGraph {nodes [
   RigVMFunctionEntryNode()
   n1 = RigVMVariableNode() {variable=Char Backwards}
   n2 = &RigVMFunction_AlphaInterp(
    InterpSpeedDecreasing=10
    InterpSpeedIncreasing=10
    Scale=1
    bClampResult=False
    Value=&RigVMFunction_MathDoubleMax(B=@n1.Value, A=RigVMVariableNode() {variable=Char Left}.Value).Result
   )
   n3 = RigVMVariableNode() {variable=Char Forward}
   n4 = &RigVMFunction_AlphaInterp(
    InterpSpeedDecreasing=10
    InterpSpeedIncreasing=10
    Scale=1
    bClampResult=False
    bMapRange=false
    Value=&RigVMFunction_MathDoubleMax(B=@n3.Value, A=RigVMVariableNode() {variable=Char Right}.Value).Result
   )
   Interpolate = /Script/RigVM.RigVMFunction_MathTransformLerp(
    A=(Rotation=(X=0,Y=0,Z=0,W=1),Translation=(X=0,Y=0,Z=0),Scale3D=(X=1,Y=1,Z=1))
    B=(Rotation=(X=0,Y=0,Z=0,W=1),Translation=(X=0.6077,Y=0,Z=0),Scale3D=(X=1,Y=1,Z=1))
    T=@n4.Result
   ) {is=struct RigVMUnitNode, template="Interpolate::Execute(in A,in B,in T,out Result)"}
   > &RigUnit_ModifyTransforms(
    ItemToModify="((Item=(Type=Control,Name=\"ref_HandIK_L_ctrl\"),Transform=(Rotation=(X=0,Y=0,Z=0,W=1),Translation=(X=-2.4878,Y=-0.5325,Z=-0.7023),Scale3D=(X=1,Y=1,Z=1))),(Item=(Type=Bone,Name=\"None\"),Transform=(Rotation=(X=0,Y=0,Z=0,W=1),Translation=(X=0,Y=0,Z=0),Scale3D=( ...(truncated)"
    Mode=OverrideLocal
    Weight=1
    WeightMinimum=0
    ItemToModify.0.Transform=&RigVMFunction_MathTransformLerp(B=(Rotation=(X=0,Y=0,Z=0,W=1),Translation=(X=-2.4878,Y=-0.5325,Z=-0.7023),Scale3D=(X=1,Y=1,Z=1)), T=@n2.Result, A=@Interpolate.Result) {template="Interpolate::Execute(in A,in B,in T,out Result)"}.Result
   )
   n5 = &RigVMFunction_MathTransformLerp(
    A=(Rotation=(X=0,Y=0,Z=0,W=1),Translation=(X=0,Y=0,Z=0),Scale3D=(X=1,Y=1,Z=1))
    B=(Rotation=(X=0,Y=0,Z=0,W=1),Translation=(X=0,Y=0,Z=0),Scale3D=(X=1,Y=1,Z=1))
    T=@n4.Result
   ) {template="Interpolate::Execute(in A,in B,in T,out Result)"}
   > &RigUnit_ModifyTransforms(
    ItemToModify=((Item=(Type=Control,Name="ref_PoleVect_L_ctrl"),Transform=(Rotation=(X=0,Y=0,Z=0,W=1),Translation=(X=0,Y=0,Z=0),Scale3D=(X=1,Y=1,Z=1))))
    Mode=OverrideLocal
    Weight=1
    WeightMinimum=0
    ItemToModify.0.Transform=&RigVMFunction_MathTransformLerp(B=(Rotation=(X=0,Y=0,Z=0,W=1),Translation=(X=0,Y=0,Z=6),Scale3D=(X=1,Y=1,Z=1)), T=@n2.Result, A=@n5.Result) {template="Interpolate::Execute(in A,in B,in T,out Result)"}.Result
   )
   n6 = &RigVMFunction_MathTransformLerp(
    A=(Rotation=(X=0,Y=0,Z=0,W=1),Translation=(X=0,Y=0,Z=0),Scale3D=(X=1,Y=1,Z=1))
    B=(Rotation=(X=-0.0436,Y=0,Z=0,W=0.999),Translation=(X=0,Y=0,Z=0),Scale3D=(X=1,Y=1,Z=1))
    T=@n4.Result
   ) {template="Interpolate::Execute(in A,in B,in T,out Result)"}
   > &RigUnit_ModifyTransforms(
    ItemToModify=((Item=(Type=Bone,Name="handlerBar_l"),Transform=(Rotation=(X=0,Y=0,Z=0,W=1),Translation=(X=0,Y=0,Z=0),Scale3D=(X=1,Y=1,Z=1))))
    Weight=1
    WeightMinimum=0
    ItemToModify.0.Transform=&RigVMFunction_MathTransformLerp(B=(Rotation=(X=0.1743,Y=0,Z=0,W=0.9847),Translation=(X=0,Y=0,Z=0),Scale3D=(X=1,Y=1,Z=1)), T=@n2.Result, A=@n6.Result) {template="Interpolate::Execute(in A,in B,in T,out Result)"}.Result
   )
   n7 = RigVMVariableNode() {variable=Char Backwards}
   n8 = &RigVMFunction_AlphaInterp(
    InterpSpeedDecreasing=10
    InterpSpeedIncreasing=10
    Scale=1
    bClampResult=False
    Value=&RigVMFunction_MathDoubleMax(B=@n7.Value, A=RigVMVariableNode() {variable=Char Right}.Value).Result
   )
   n9 = RigVMVariableNode() {variable=Char Forward}
   n10 = &RigVMFunction_AlphaInterp(
    InterpSpeedDecreasing=10
    InterpSpeedIncreasing=10
    Scale=1
    bClampResult=False
    bMapRange=false
    Value=&RigVMFunction_MathDoubleMax(A=RigVMVariableNode() {variable=Char Left}.Value, B=@n9.Value).Result
   )
   n11 = &RigVMFunction_MathTransformLerp(
    A=(Rotation=(X=0,Y=0,Z=0,W=1),Translation=(X=0,Y=0,Z=0),Scale3D=(X=1,Y=1,Z=1))
    B=(Rotation=(X=0,Y=0,Z=0,W=1),Translation=(X=0.6077,Y=0,Z=0),Scale3D=(X=1,Y=1,Z=1))
    T=@n10.Result
   ) {template="Interpolate::Execute(in A,in B,in T,out Result)"}
   > &RigUnit_ModifyTransforms(
    ItemToModify=((Item=(Type=Control,Name="ref_HandIK_R_ctrl"),Transform=(Rotation=(X=0,Y=0,Z=0,W=1),Translation=(X=0,Y=0,Z=0),Scale3D=(X=1,Y=1,Z=1))),(Item=(Type=Bone,Name="None"),Transform=(Rotation=(X=0,Y=0,Z=0,W=1),Translation=(X=0,Y=0,Z=0),Scale3D=(X=1,Y=1,Z=1))))
    Mode=OverrideLocal
    Weight=1
    WeightMinimum=0
    ItemToModify.0.Transform=&RigVMFunction_MathTransformLerp(B=(Rotation=(X=0,Y=0,Z=0,W=1),Translation=(X=-2.4878,Y=0.5325,Z=-0.7023),Scale3D=(X=1,Y=1,Z=1)), T=@n8.Result, A=@n11.Result) {template="Interpolate::Execute(in A,in B,in T,out Result)"}.Result
   )
   n12 = &RigVMFunction_MathTransformLerp(
    A=(Rotation=(X=0,Y=0,Z=0,W=1),Translation=(X=0,Y=0,Z=0),Scale3D=(X=1,Y=1,Z=1))
    B=(Rotation=(X=0,Y=0,Z=0,W=1),Translation=(X=0,Y=0,Z=0),Scale3D=(X=1,Y=1,Z=1))
    T=@n10.Result
   ) {template="Interpolate::Execute(in A,in B,in T,out Result)"}
   > &RigUnit_ModifyTransforms(
    ItemToModify=((Item=(Type=Control,Name="ref_PoleVect_R_ctrl"),Transform=(Rotation=(X=0,Y=0,Z=0,W=1),Translation=(X=0,Y=0,Z=0),Scale3D=(X=1,Y=1,Z=1))))
    Mode=OverrideLocal
    Weight=1
    WeightMinimum=0
    ItemToModify.0.Transform=&RigVMFunction_MathTransformLerp(B=(Rotation=(X=0,Y=0,Z=0,W=1),Translation=(X=0,Y=0,Z=6),Scale3D=(X=1,Y=1,Z=1)), T=@n8.Result, A=@n12.Result) {template="Interpolate::Execute(in A,in B,in T,out Result)"}.Result
   )
   n13 = &RigVMFunction_MathTransformLerp(
    A=(Rotation=(X=0,Y=0,Z=0,W=1),Translation=(X=0,Y=0,Z=0),Scale3D=(X=1,Y=1,Z=1))
    B=(Rotation=(X=0.0436,Y=0,Z=0,W=0.999),Translation=(X=0,Y=0,Z=0),Scale3D=(X=1,Y=1,Z=1))
    T=@n10.Result
   ) {template="Interpolate::Execute(in A,in B,in T,out Result)"}
   > &RigUnit_ModifyTransforms(
    ItemToModify=((Item=(Type=Bone,Name="handlerBar_r"),Transform=(Rotation=(X=0,Y=0,Z=0,W=1),Translation=(X=0,Y=0,Z=0),Scale3D=(X=1,Y=1,Z=1))))
    Weight=1
    WeightMinimum=0
    ItemToModify.0.Transform=&RigVMFunction_MathTransformLerp(B=(Rotation=(X=-0.1743,Y=0,Z=0,W=0.9847),Translation=(X=0,Y=0,Z=0),Scale3D=(X=1,Y=1,Z=1)), T=@n8.Result, A=@n13.Result) {template="Interpolate::Execute(in A,in B,in T,out Result)"}.Result
   )
   n14 = RigVMVariableNode() {variable=Head Down}
   n15 = &RigVMFunction_AlphaInterp(
    InterpSpeedDecreasing=10
    InterpSpeedIncreasing=10
    Scale=1
    bClampResult=False
    Value=&RigVMFunction_MathDoubleMax(B=RigVMVariableNode() {variable=Head Left}.Value, A=@n14.Value).Result
   )
   > &RigUnit_ModifyTransforms(
    ItemToModify="((Item=(Type=Control,Name=\"ref_HandIK_L_ctrl\"),Transform=(Rotation=(X=0,Y=0,Z=0,W=1),Translation=(X=-3.6096,Y=-1.1493,Z=-1.3526),Scale3D=(X=1,Y=1,Z=1))),(Item=(Type=Bone,Name=\"None\"),Transform=(Rotation=(X=0,Y=0,Z=0,W=1),Translation=(X=0,Y=0,Z=0),Scale3D=( ...(truncated)"
    Mode=OverrideLocal
    WeightMinimum=0
    Weight=@n15.Result
   )
   > &RigUnit_ModifyTransforms(
    ItemToModify=((Item=(Type=Control,Name="ref_PoleVect_L_ctrl"),Transform=(Rotation=(X=0,Y=0,Z=0,W=1),Translation=(X=0,Y=0,Z=7.8101),Scale3D=(X=1,Y=1,Z=1))))
    Mode=OverrideLocal
    WeightMinimum=0
    Weight=@n15.Result
   )
   > &RigUnit_ModifyTransforms(
    ItemToModify=((Item=(Type=Bone,Name="handlerBar_l"),Transform=(Rotation=(X=0.2588,Y=0,Z=0,W=0.9659),Translation=(X=0,Y=0,Z=0),Scale3D=(X=1,Y=1,Z=1))))
    WeightMinimum=0
    Weight=@n15.Result
   )
   n16 = &RigVMFunction_AlphaInterp(
    InterpSpeedDecreasing=10
    InterpSpeedIncreasing=10
    Scale=1
    bClampResult=False
    Value=&RigVMFunction_MathDoubleMax(B=RigVMVariableNode() {variable=Head Right}.Value, A=@n14.Value).Result
   )
   > &RigUnit_ModifyTransforms(
    ItemToModify="((Item=(Type=Control,Name=\"ref_HandIK_R_ctrl\"),Transform=(Rotation=(X=0,Y=0,Z=0,W=1),Translation=(X=-3.6096,Y=1.1493,Z=-1.3526),Scale3D=(X=1,Y=1,Z=1))),(Item=(Type=Bone,Name=\"None\"),Transform=(Rotation=(X=0,Y=0,Z=0,W=1),Translation=(X=0,Y=0,Z=0),Scale3D=(X ...(truncated)"
    Mode=OverrideLocal
    WeightMinimum=0
    Weight=@n16.Result
   )
   > &RigUnit_ModifyTransforms(
    ItemToModify=((Item=(Type=Control,Name="ref_PoleVect_R_ctrl"),Transform=(Rotation=(X=0,Y=0,Z=0,W=1),Translation=(X=0,Y=0,Z=7.8101),Scale3D=(X=1,Y=1,Z=1))))
    Mode=OverrideLocal
    WeightMinimum=0
    Weight=@n16.Result
   )
   > &RigUnit_ModifyTransforms(
    ItemToModify=((Item=(Type=Bone,Name="handlerBar_r"),Transform=(Rotation=(X=-0.2588,Y=0,Z=0,W=0.9659),Translation=(X=0,Y=0,Z=0),Scale3D=(X=1,Y=1,Z=1))))
    WeightMinimum=0
    Weight=@n16.Result
   )
   > RigVMFunctionReturnNode()
   Maximum = &RigVMFunction_MathDoubleMax(
    B=RigVMDispatchNode(BoneName=root, Default=0, Name=deformRig_l_upperarm_fr_alpha) {template="DISPATCH_RigDispatch_GetAnimAttribute(in Name,in BoneName,in Default,out Value,out Found)"}.Value
    A=@n3.Value
   )
   &RigVMFunction_MathDoubleMax(
    B=RigVMDispatchNode(BoneName=root, Default=0, Name=deformRig_l_upperarm_fr_alpha) {template="DISPATCH_RigDispatch_GetAnimAttribute(in Name,in BoneName,in Default,out Value,out Found)"}.Value
    A=@n1.Value
   )
   &RigVMFunction_MathDoubleMax(
    B=RigVMDispatchNode(BoneName=root, Default=0, Name=deformRig_r_upperarm_fr_alpha) {template="DISPATCH_RigDispatch_GetAnimAttribute(in Name,in BoneName,in Default,out Value,out Found)"}.Value
    A=@n9.Value
   )
   &RigVMFunction_MathDoubleMax(
    B=RigVMDispatchNode(BoneName=root, Default=0, Name=deformRig_r_upperarm_fr_alpha) {template="DISPATCH_RigDispatch_GetAnimAttribute(in Name,in BoneName,in Default,out Value,out Found)"}.Value
    A=@n7.Value
   )
  ]}
 RigVMFunctionLibrary {nodes [
   Kenny Arm Movement = RigVMCollapseNode(ExecuteContext=())
   Kenny Arms IK Setup = RigVMCollapseNode(ExecuteContext=())
  ]}
 RigVMModel {nodes [
   n1 = &RigUnit_PrepareForExecution()
   Control Stack at Item_4 = RigVMFunctionReferenceNode(
    Add Bottom Null=False
    Add Null=true
    Add Secondary Control=False
    Control Suffix=_ctrl
    Item=(Name="head",Type=Bone)
    Name=ref_head_01
    Null Suffix=_null
    Parent=(Name="neck_02",Type=Bone)
    Secondary Settings="(Shape=(bVisible=True,Name=\"Default\",Color=(R=1,G=0,B=0,A=1),Transform=(Rotation=(X=0,Y=0,Z=0,W=1),Translation=(X=0,Y=0,Z=0),Scale3D=(X=1,Y=1,Z=1))),Proxy=(bIsProxy=False,ShapeVisibility=BasedOnSelection),DisplayName=\"None\",InitialSpace=LocalSpace,bUsePref ...(truncated)"
    Settings="(Shape=(bVisible=True,Name=\"Pyramid_Thick\",Color=(R=0,G=0,B=1,A=1),Transform=(Rotation=(X=0,Y=0,Z=0,W=1),Translation=(X=0,Y=0,Z=0),Scale3D=(X=1,Y=1,Z=1))),Proxy=(bIsProxy=False,ShapeVisibility=BasedOnSelection),DisplayName=\"None\",InitialSpace=LocalSpace,bU ...(truncated)"
    ExecuteContext=@n1.ExecutePin
   )
   > Control Stack at Item_1 = RigVMFunctionReferenceNode(
    Add Bottom Null=False
    Add Null=true
    Add Secondary Control=False
    Control Suffix=_ctrl
    Item=(Name="neck_01",Type=Bone)
    Name=ref_neck_01
    Null Suffix=_null
    Parent=(Name="spine_05",Type=Bone)
    Secondary Settings="(Shape=(bVisible=True,Name=\"Default\",Color=(R=1,G=0,B=0,A=1),Transform=(Rotation=(X=0,Y=0,Z=0,W=1),Translation=(X=0,Y=0,Z=0),Scale3D=(X=1,Y=1,Z=1))),Proxy=(bIsProxy=False,ShapeVisibility=BasedOnSelection),DisplayName=\"None\",InitialSpace=LocalSpace,bUsePref ...(truncated)"
    Settings="(Shape=(bVisible=True,Name=\"Pyramid_Thick\",Color=(R=0,G=0,B=1,A=1),Transform=(Rotation=(X=0,Y=0,Z=0,W=1),Translation=(X=0,Y=0,Z=0),Scale3D=(X=1,Y=1,Z=1))),Proxy=(bIsProxy=False,ShapeVisibility=BasedOnSelection),DisplayName=\"None\",InitialSpace=LocalSpace,bU ...(truncated)"
   )
   > Control Stack at Item_7 = RigVMFunctionReferenceNode(
    Add Bottom Null=true
    Add Null=true
    Add Secondary Control=False
    Control Suffix=_ctrl
    Item=(Name="dyn_kennyWrist_l",Type=Bone)
    Name=ref_HandIK_L
    Null Suffix=_null
    Parent=(Name="spine_05",Type=Bone)
    Secondary Settings="(Shape=(bVisible=True,Name=\"Default\",Color=(R=0,G=0,B=1,A=1),Transform=(Rotation=(X=0,Y=0,Z=0,W=1),Translation=(X=0,Y=0,Z=0),Scale3D=(X=1,Y=1,Z=1))),Proxy=(bIsProxy=False,ShapeVisibility=BasedOnSelection),DisplayName=\"None\",InitialSpace=LocalSpace,bUsePref ...(truncated)"
    Settings="(Shape=(bVisible=True,Name=\"Box_Thick\",Color=(R=0,G=0,B=1,A=1),Transform=(Rotation=(X=0,Y=0,Z=0,W=1),Translation=(X=0,Y=0,Z=0),Scale3D=(X=0.5,Y=0.5,Z=0.5))),Proxy=(bIsProxy=False,ShapeVisibility=BasedOnSelection),DisplayName=\"None\",InitialSpace=LocalSpace, ...(truncated)"
   )
   > Control Stack at Item_8 = RigVMFunctionReferenceNode(
    Add Bottom Null=true
    Add Null=true
    Add Secondary Control=False
    Control Suffix=_ctrl
    Item=(Name="dyn_kennyWrist_r",Type=Bone)
    Name=ref_HandIK_R
    Null Suffix=_null
    Parent=(Name="spine_05",Type=Bone)
    Secondary Settings="(Shape=(bVisible=True,Name=\"Default\",Color=(R=1,G=0,B=0,A=1),Transform=(Rotation=(X=0,Y=0,Z=0,W=1),Translation=(X=0,Y=0,Z=0),Scale3D=(X=1,Y=1,Z=1))),Proxy=(bIsProxy=False,ShapeVisibility=BasedOnSelection),DisplayName=\"None\",InitialSpace=LocalSpace,bUsePref ...(truncated)"
    Settings="(Shape=(bVisible=True,Name=\"Box_Thick\",Color=(R=1,G=0,B=0,A=1),Transform=(Rotation=(X=0,Y=0,Z=0,W=1),Translation=(X=0,Y=0,Z=0),Scale3D=(X=0.5,Y=0.5,Z=0.5))),Proxy=(bIsProxy=False,ShapeVisibility=BasedOnSelection),DisplayName=\"None\",InitialSpace=LocalSpace, ...(truncated)"
   )
   > Control Stack at Item_9 = RigVMFunctionReferenceNode(
    Add Bottom Null=true
    Add Null=true
    Add Secondary Control=False
    Control Suffix=_ctrl
    Item=(Name="dyn_kennyLowerarm_l",Type=Bone)
    Name=ref_PoleVect_L
    Null Suffix=_null
    Parent=(Name="spine_05",Type=Bone)
    Secondary Settings="(Shape=(bVisible=True,Name=\"Default\",Color=(R=0,G=0,B=1,A=1),Transform=(Rotation=(X=0,Y=0,Z=0,W=1),Translation=(X=0,Y=0,Z=0),Scale3D=(X=1,Y=1,Z=1))),Proxy=(bIsProxy=False,ShapeVisibility=BasedOnSelection),DisplayName=\"None\",InitialSpace=LocalSpace,bUsePref ...(truncated)"
    Settings="(Shape=(bVisible=True,Name=\"Pyramid_Thick\",Color=(R=0,G=0,B=1,A=1),Transform=(Rotation=(X=0,Y=0,Z=0.7071,W=0.7071),Translation=(X=0,Y=0,Z=0),Scale3D=(X=0.5,Y=0.5,Z=0.5))),Proxy=(bIsProxy=False,ShapeVisibility=BasedOnSelection),DisplayName=\"None\",InitialSpa ...(truncated)"
   )
   > Control Stack at Item_10 = RigVMFunctionReferenceNode(
    Add Bottom Null=true
    Add Null=true
    Add Secondary Control=False
    Control Suffix=_ctrl
    Item=(Name="dyn_kennyLowerarm_r",Type=Bone)
    Name=ref_PoleVect_R
    Null Suffix=_null
    Parent=(Name="spine_05",Type=Bone)
    Secondary Settings="(Shape=(bVisible=True,Name=\"Default\",Color=(R=0,G=0,B=1,A=1),Transform=(Rotation=(X=0,Y=0,Z=0,W=1),Translation=(X=0,Y=0,Z=0),Scale3D=(X=1,Y=1,Z=1))),Proxy=(bIsProxy=False,ShapeVisibility=BasedOnSelection),DisplayName=\"None\",InitialSpace=LocalSpace,bUsePref ...(truncated)"
    Settings="(Shape=(bVisible=True,Name=\"Pyramid_Thick\",Color=(R=1,G=0,B=0,A=1),Transform=(Rotation=(X=0,Y=0,Z=-0.7071,W=0.7071),Translation=(X=0,Y=0,Z=0),Scale3D=(X=0.5,Y=0.5,Z=0.5))),Proxy=(bIsProxy=False,ShapeVisibility=BasedOnSelection),DisplayName=\"None\",InitialSp ...(truncated)"
   )
   > &RigUnit_ModifyTransforms(
    ItemToModify=((Item=(Type=Null,Name="ref_PoleVect_L_ctrl_null"),Transform=(Rotation=(X=0,Y=0,Z=0,W=1),Translation=(X=0,Y=-10,Z=0),Scale3D=(X=1,Y=1,Z=1))))
    Mode=OverrideLocal
    Weight=1
    WeightMinimum=0
   )
   > &RigUnit_ModifyTransforms(
    ItemToModify=((Item=(Type=Null,Name="ref_PoleVect_R_ctrl_null"),Transform=(Rotation=(X=0,Y=0,Z=0,W=1),Translation=(X=0,Y=10,Z=0),Scale3D=(X=1,Y=1,Z=1))))
    Mode=OverrideLocal
    Weight=1
    WeightMinimum=0
   )
   RigUnit_BeginExecution = &RigUnit_BeginExecution()
   FN_DeformRig_FNC = RigVMFunctionReferenceNode(
    DeformRig_ Curve="(deformRig_l_upperarm_up_alpha_72_#1=true,deformRig_r_upperarm_up_alpha_73_#2=true,deformRig_l_upperarm_fr_alpha_74_#3=true,deformRig_r_upperarm_fr_alpha_75_CB51B80C4 ...(truncated)"
    DeformRig_Bool="(upperarm_1_#4=True,elbow_3_#5=True,wrist_7_#6=True,groin_8_#7=True,glute_17_#8=True,knee_9_40C408324CA261 ...(truncated)"
    DeformRig_Debug=(upperarm_debug_18_#4=False,wrist_debug_20_#6=False,lat_debug_26_#9=False,groinGlute_debug_27_#7=False)
    DeformRig_Value="(upperarm_up_18_#10=(Rotation=(X=0.0,Y=-0.0,Z=0.0,W=1.0),Translation=(X=0.0,Y=0.0,Z=8.0),Scale3D=(X=0.70,Y=1.0,Z=1.0)),upperarm_fwd_fr_19_#11=(Rotation=(X=0.0,Y=-0.0,Z=0.0,W=1.0),Translation=(X=0.0, ...(truncated)"
    ExecuteContext=@RigUnit_BeginExecution.ExecutePin
   )
   > Kenny Arms IK Setup = RigVMFunctionReferenceNode()
   > Kenny Arm Movement = RigVMFunctionReferenceNode()
  ]}
]
hierarchy [
 root {loc=<0 0 0>, rot=<0 0 0 1>}
 pelvis {parent=root, loc=<0 -0.7321 92.4218>, rot=<0.0127 0.707 -0.0127 -0.707>}
 spine_01 {parent=pelvis, loc=<5.6002 -0.5128 0>, rot=<0 0 0.034 -0.9994>}
 spine_02 {parent=spine_01, loc=<8.0756 0.0139 0>, rot=<0 0 0.0238 -0.9997>}
 spine_03 {parent=spine_02, loc=<7.0395 0 0>, rot=<0 0 0.0149 -0.9999>}
 spine_04 {parent=spine_03, loc=<8.1023 0 0>, rot=<0 0 -0.0132 -0.9999>}
 spine_05 {parent=spine_04, loc=<10.0141 0 0>, rot=<0 0 -0.0015 -1>}
 clavicle_l {parent=spine_05, loc=<2.5856 -0.2866 -4.3777>, rot=<0.0465 -0.7289 -0.009 -0.6829>}
 upperarm_l {parent=clavicle_l, loc=<18.2264 0.1342 -0.001>, rot=<0.0253 -0.2693 0.0013 -0.9627>}
 lowerarm_l {parent=upperarm_l, loc=<29.0071 0.1686 -1.239>, rot=<0.0391 -0.0099 -0.2453 -0.9686>}
 hand_l {parent=lowerarm_l, loc=<25.8607 0.222 1.0274>, rot=<-0.7193 -0.0988 0.1193 -0.6773>}
 weapon_l {parent=hand_l, loc=<-0.1974 -3.4111 0.3811>, rot=<0 0 0 1>}
 hand_attach_l {parent=hand_l, loc=<8.5 -3 0>, rot=<0.7071 0.7071 0 0>}
 thumb_01_l {parent=hand_l, loc=<1.3118 -0.3191 -2.459>, rot=<0.6546 -0.4042 0.0664 -0.6354>}
 thumb_02_l {parent=thumb_01_l, loc=<5.9336 -0.0448 0.0401>, rot=<-0.0037 0.0213 0.298 -0.9543>}
 thumb_03_l {parent=thumb_02_l, loc=<2.2223 0.3682 -0.0907>, rot=<0.0118 -0.013 0.0066 -0.9998>}
 index_metacarpal_l {parent=hand_l, loc=<1.6535 0.0183 -2.96>, rot=<0.0158 -0.0124 -0.0173 -0.9996>}
 index_01_l {parent=index_metacarpal_l, loc=<8.3008 0.0357 -0.0524>, rot=<-0.0044 0.0485 0.0924 -0.9945>}
 index_02_l {parent=index_01_l, loc=<3.3587 -0.3004 0.0452>, rot=<-0.002 0.0027 0.3225 -0.9466>}
 index_03_l {parent=index_02_l, loc=<2.5578 -0.123 -0.0252>, rot=<-0.002 0.0067 0.2558 -0.9667>}
 middle_metacarpal_l {parent=hand_l, loc=<2.1856 -0.1725 -0.5984>, rot=<0.006 0.0168 -0.0329 -0.9993>}
 middle_01_l {parent=middle_metacarpal_l, loc=<7.3275 -0.2447 0.0129>, rot=<-0.0208 0.0042 0.0858 -0.9961>}
 middle_02_l {parent=middle_01_l, loc=<4.0123 0.1401 0.1175>, rot=<-0.0218 0.0258 0.3512 -0.9357>}
 middle_03_l {parent=middle_02_l, loc=<2.6077 0.0534 0.1373>, rot=<0.0082 0.0188 0.1708 -0.9851>}
 ring_metacarpal_l {parent=hand_l, loc=<2.2074 -0.3625 1.3638>, rot=<-0.0284 0.0722 -0.0458 -0.9959>}
 ring_01_l {parent=ring_metacarpal_l, loc=<7.0295 -0.6197 -0.1446>, rot=<-0.0489 -0.0252 0.1892 -0.9804>}
 ring_02_l {parent=ring_01_l, loc=<3.6174 0.4647 -0.1027>, rot=<0 0 0.2806 -0.9598>}
 ring_03_l {parent=ring_02_l, loc=<2.4152 0.114 0.0199>, rot=<0 0 0.1189 -0.9929>}
 pinky_metacarpal_l {parent=hand_l, loc=<1.7616 -0.9753 3.3814>, rot=<-0.0397 0.0814 -0.069 -0.9935>}
 pinky_01_l {parent=pinky_metacarpal_l, loc=<7.1709 -0.4831 -0.1094>, rot=<-0.0654 -0.0035 0.2242 -0.9723>}
 pinky_02_l {parent=pinky_01_l, loc=<2.8274 0.0893 -0.1043>, rot=<0.0037 0.0011 0.2854 -0.9584>}
 pinky_03_l {parent=pinky_02_l, loc=<2.2046 0.0307 -0.0078>, rot=<-0.0002 0.0098 0.1445 -0.9895>}
 deform_wrist_dn_base_l {parent=hand_l, loc=<0 -3.4173 0>, rot=<0 0 0 1>}
 deform_wrist_dn_l {parent=deform_wrist_dn_base_l, loc=<0 0 0>, rot=<0 0 0 1>}
 deform_wrist_up_base_l {parent=hand_l, loc=<0 2.7805 0>, rot=<0 0 0 1>}
 deform_wrist_up_l {parent=deform_wrist_up_base_l, loc=<0 0 0>, rot=<0 0 0 1>}
 lowerarm_twist_02_l {parent=lowerarm_l, loc=<9.6199 0 0>, rot=<0 0 0 -1>}
 lowerarm_twist_01_l {parent=lowerarm_l, loc=<17.8549 0 0>, rot=<0 0 0 -1>}
 deform_elbow_in_base_l {parent=lowerarm_l, loc=<0.2112 5.645 0.8274>, rot=<0 0 0 -1>}
 deform_elbow_in_l {parent=deform_elbow_in_base_l, loc=<0 0 0>, rot=<0 0 0 1>}
 deform_elbow_out_base_l {parent=lowerarm_l, loc=<0 -7.5129 0>, rot=<0 0 0 -1>}
 deform_elbow_out_l {parent=deform_elbow_out_base_l, loc=<0 0 0>, rot=<0 0 0 1>}
 clawMaster_l {parent=lowerarm_l, loc=<26.8446 0.0096 -0.0864>, rot=<-0.7574 -0.019 0.0495 -0.6508>}
 claw_01_l {parent=clawMaster_l, loc=<4.7643 0.3577 -0.0413>, rot=<0.0406 -0.023 -0.4363 -0.8986>}
 claw_02_l {parent=clawMaster_l, loc=<4.7629 -0.407 0.0411>, rot=<-0.0462 -0.0042 0.4841 -0.8738>}
 dyn_lowerarm_base_l {parent=lowerarm_l, loc=<0 0 0>, rot=<0 0 0 1>}
 upperarm_twist_01_l {parent=upperarm_l, loc=<7.9796 0.0843 -0.6195>, rot=<0 0 0 -1>}
 deform_upperarm_bk_base_l {parent=upperarm_twist_01_l, loc=<-6.8983 -13.3826 -1.3258>, rot=<0 0 0 1>}
 deform_upperarm_bk_l {parent=deform_upperarm_bk_base_l, loc=<0 0 0>, rot=<0 0 0 1>}
 deform_upperarm_fr_base_l {parent=upperarm_twist_01_l, loc=<-8.3487 10.7117 -2.1455>, rot=<0 0 0 1>}
 deform_upperarm_fr_l {parent=deform_upperarm_fr_base_l, loc=<0 0 0>, rot=<0 0 0 1>}
 deform_upperarm_up_base_l {parent=upperarm_twist_01_l, loc=<-11.2012 -1.2234 8.7784>, rot=<0 0 0 1>}
 deform_upperarm_up_l {parent=deform_upperarm_up_base_l, loc=<0 0 0>, rot=<0 0 0 1>}
 upperarm_twist_02_l {parent=upperarm_l, loc=<19.5059 0.0843 -0.6195>, rot=<0 0 0 -1>}
 upperarmMover_l {parent=upperarm_l, loc=<23.708 0.0843 -0.6195>, rot=<0 0 0 -1>}
 clavicleMover_l {parent=clavicle_l, loc=<0 0 0>, rot=<0 0 0 1>}
 neck_01 {parent=spine_05, loc=<10.815 1.9196 0>, rot=<0 0 -0.1864 -0.9825>}
 neck_02 {parent=neck_01, loc=<4.7347 0.3045 0>, rot=<0 0 0.0363 -0.9993>}
 head {parent=neck_02, loc=<5.7632 0.5552 0>, rot=<0 0 0.115 -0.9934>}
 clavicle_r {parent=spine_05, loc=<2.5855 -0.2866 4.3777>, rot=<0.7289 0.0465 -0.6829 0.009>}
 upperarm_r {parent=clavicle_r, loc=<-18.2264 -0.1342 0.001>, rot=<0.0253 -0.2693 0.0013 -0.9627>}
 lowerarm_r {parent=upperarm_r, loc=<-29.0071 -0.1686 1.239>, rot=<0.0391 -0.0099 -0.2453 -0.9686>}
 hand_r {parent=lowerarm_r, loc=<-25.8607 -0.222 -1.0274>, rot=<-0.7193 -0.0988 0.1193 -0.6773>}
 hand_attach_r {parent=hand_r, loc=<-8.5 3 0>, rot=<0 0 -0.7071 -0.7071>}
 thumb_01_r {parent=hand_r, loc=<-1.3118 0.3191 2.459>, rot=<0.6546 -0.4042 0.0664 -0.6354>}
 thumb_02_r {parent=thumb_01_r, loc=<-5.9336 0.0448 -0.0401>, rot=<-0.0037 0.0213 0.298 -0.9543>}
 thumb_03_r {parent=thumb_02_r, loc=<-2.2223 -0.3682 0.0907>, rot=<0.0118 -0.013 0.0066 -0.9998>}
 index_metacarpal_r {parent=hand_r, loc=<-1.6535 -0.0183 2.96>, rot=<0.0158 -0.0124 -0.0173 -0.9996>}
 index_01_r {parent=index_metacarpal_r, loc=<-8.3008 -0.0357 0.0524>, rot=<-0.0044 0.0485 0.0924 -0.9945>}
 index_02_r {parent=index_01_r, loc=<-3.3587 0.3004 -0.0452>, rot=<-0.002 0.0027 0.3225 -0.9466>}
 index_03_r {parent=index_02_r, loc=<-2.5578 0.123 0.0252>, rot=<-0.002 0.0067 0.2558 -0.9667>}
 middle_metacarpal_r {parent=hand_r, loc=<-2.1856 0.1725 0.5984>, rot=<0.006 0.0168 -0.0329 -0.9993>}
 middle_01_r {parent=middle_metacarpal_r, loc=<-7.3275 0.2447 -0.0129>, rot=<-0.0208 0.0042 0.0858 -0.9961>}
 middle_02_r {parent=middle_01_r, loc=<-4.0123 -0.1401 -0.1175>, rot=<-0.0218 0.0258 0.3512 -0.9357>}
 middle_03_r {parent=middle_02_r, loc=<-2.6077 -0.0534 -0.1373>, rot=<0.0082 0.0188 0.1708 -0.9851>}
 ring_metacarpal_r {parent=hand_r, loc=<-2.2074 0.3625 -1.3638>, rot=<-0.0284 0.0722 -0.0458 -0.9959>}
 ring_01_r {parent=ring_metacarpal_r, loc=<-7.0295 0.6197 0.1446>, rot=<-0.0489 -0.0252 0.1892 -0.9804>}
 ring_02_r {parent=ring_01_r, loc=<-3.6174 -0.4647 0.1027>, rot=<0 0 0.2806 -0.9598>}
 ring_03_r {parent=ring_02_r, loc=<-2.4152 -0.114 -0.0199>, rot=<0 0 0.1189 -0.9929>}
 pinky_metacarpal_r {parent=hand_r, loc=<-1.7616 0.9753 -3.3814>, rot=<-0.0397 0.0814 -0.069 -0.9935>}
 pinky_01_r {parent=pinky_metacarpal_r, loc=<-7.1709 0.4831 0.1094>, rot=<-0.0654 -0.0035 0.2242 -0.9723>}
 pinky_02_r {parent=pinky_01_r, loc=<-2.8274 -0.0893 0.1043>, rot=<0.0037 0.0011 0.2854 -0.9584>}
 pinky_03_r {parent=pinky_02_r, loc=<-2.2046 -0.0307 0.0078>, rot=<-0.0002 0.0098 0.1445 -0.9895>}
 weapon_r {parent=hand_r, loc=<0.1974 3.4111 -0.3811>, rot=<0 0 0 1>}
 deform_wrist_dn_base_r {parent=hand_r, loc=<0 3.4172 0>, rot=<0 0 0 -1>}
 deform_wrist_dn_r {parent=deform_wrist_dn_base_r, loc=<0 0 0>, rot=<0 0 0 -1>}
 deform_wrist_up_base_r {parent=hand_r, loc=<-0.0001 -2.7804 0>, rot=<0 0 0 -1>}
 deform_wrist_up_r {parent=deform_wrist_up_base_r, loc=<0 0 0>, rot=<0 0 0 -1>}
 lowerarm_twist_02_r {parent=lowerarm_r, loc=<-9.6199 0 0>, rot=<0 0 0 -1>}
 lowerarm_twist_01_r {parent=lowerarm_r, loc=<-17.8549 0 0>, rot=<0 0 0 -1>}
 deform_elbow_in_base_r {parent=lowerarm_r, loc=<-0.2114 -5.6449 -0.827>, rot=<0 0 0 -1>}
 deform_elbow_in_r {parent=deform_elbow_in_base_r, loc=<0 0 0>, rot=<0 0 0 1>}
 deform_elbow_out_base_r {parent=lowerarm_r, loc=<-0.0001 7.5129 0.0002>, rot=<0 0 0 -1>}
 deform_elbow_out_r {parent=deform_elbow_out_base_r, loc=<0 0 0>, rot=<0 0 0 1>}
 clawMaster_r {parent=lowerarm_r, loc=<-26.8446 -0.0096 0.0864>, rot=<0.0495 0.6508 0.7574 -0.019>}
 claw_01_r {parent=clawMaster_r, loc=<4.7643 -0.3577 -0.0413>, rot=<0.8986 -0.4363 -0.023 -0.0406>}
 claw_02_r {parent=clawMaster_r, loc=<4.7628 0.4069 0.0411>, rot=<0.8738 0.4841 -0.0042 0.0462>}
 lowerarmMover_r {parent=lowerarm_r, loc=<-2.9774 0 0>, rot=<0 0 0 -1>}
 dyn_lowerarm_base_r {parent=lowerarm_r, loc=<0 0 0>, rot=<0 0 0 1>}
 upperarm_twist_01_r {parent=upperarm_r, loc=<-7.9796 -0.0843 0.6195>, rot=<0 0 0 -1>}
 deform_upperarm_bk_base_r {parent=upperarm_twist_01_r, loc=<6.8985 13.3826 1.3257>, rot=<0 0 0 1>}
 deform_upperarm_bk_r {parent=deform_upperarm_bk_base_r, loc=<0 0 0>, rot=<0 0 0 1>}
 deform_upperarm_fr_base_r {parent=upperarm_twist_01_r, loc=<8.3485 -10.7117 2.1456>, rot=<0 0 0 1>}
 deform_upperarm_fr_r {parent=deform_upperarm_fr_base_r, loc=<0 0 0>, rot=<0 0 0 1>}
 deform_upperarm_up_base_r {parent=upperarm_twist_01_r, loc=<11.2015 1.2234 -8.7788>, rot=<0 0 0 1>}
 deform_upperarm_up_r {parent=deform_upperarm_up_base_r, loc=<0 0 0>, rot=<0 0 0 1>}
 upperarm_twist_02_r {parent=upperarm_r, loc=<-19.5059 -0.0843 0.6195>, rot=<0 0 0 -1>}
 clavicleMover_r {parent=clavicle_r, loc=<0 0 0>, rot=<0 0 0 1>}
 deform_pelvisBaby {parent=spine_05, loc=<5.3989 6.0322 0>, rot=<0 0 -0.0822 -0.9966>}
 deform_spine1Baby {parent=deform_pelvisBaby, loc=<4.1432 0 0>, rot=<0 0 0.0218 -0.9998>}
 deform_headBaby {parent=deform_spine1Baby, loc=<7.6416 0 0>, rot=<0 0 0 1>}
 hoodStringMaster {parent=deform_headBaby, loc=<4.6992 17.4139 0>, rot=<-0.0143 -0.707 0.0143 -0.707>}
 base_hoodString_01_r {parent=hoodStringMaster, loc=<0.2712 -0.006 -0.0312>, rot=<0.6919 0.0583 -0.6004 0.3967>}
 dyn_hoodString_01_r {parent=base_hoodString_01_r, loc=<0 0 0>, rot=<0 0 0 1>}
 dyn_hoodString_02_r {parent=dyn_hoodString_01_r, loc=<1.19 0 0>, rot=<0.9054 -0.2186 0.0854 -0.3538>}
 dyn_hoodString_03_r {parent=dyn_hoodString_02_r, loc=<2.2008 0 0>, rot=<0.0338 0.0154 -0.0697 -0.9969>}
 dyn_hoodString_04_r {parent=dyn_hoodString_03_r, loc=<2.6074 0 0>, rot=<0.0041 0.0069 -0.0083 -0.9999>}
 dyn_hoodString_05_r {parent=dyn_hoodString_04_r, loc=<2.6593 0 0>, rot=<0 0 0 1>}
 base_hoodString_01_l {parent=hoodStringMaster, loc=<-0.2712 -0.006 -0.0312>, rot=<-0.0802 -0.7078 0.3142 -0.6276>}
 dyn_hoodString_01_l {parent=base_hoodString_01_l, loc=<0 0 0>, rot=<0 0 0 1>}
 dyn_hoodString_02_l {parent=dyn_hoodString_01_l, loc=<1.4331 0 0>, rot=<0.2526 0.05 -0.1878 -0.9479>}
 dyn_hoodString_03_l {parent=dyn_hoodString_02_l, loc=<1.9686 0 0>, rot=<0.0256 -0.0057 -0.0902 -0.9956>}
 dyn_hoodString_04_l {parent=dyn_hoodString_03_l, loc=<2.1257 -0.0002 -0.0245>, rot=<0.0052 -0.0144 0.0257 -0.9996>}
 dyn_hoodString_05_l {parent=dyn_hoodString_04_l, loc=<2.0701 0 0>, rot=<0 0 0 1>}
 hoodStringsMaster {parent=deform_headBaby, loc=<4.7536 22.857 0>, rot=<-0.0143 -0.707 0.0143 -0.707>}
 base_hoodStrings_01_l {parent=hoodStringsMaster, loc=<0.4801 0.1103 -0.1752>, rot=<0.6067 0.4295 -0.5037 0.4401>}
 dyn_hoodStrings_01_l {parent=base_hoodStrings_01_l, loc=<0 0 0>, rot=<0 0 0 1>}
 dyn_hoodStrings_02_l {parent=dyn_hoodStrings_01_l, loc=<2.2328 0 0>, rot=<0 0 0.031 -0.9995>}
 dyn_hoodStrings_03_l {parent=dyn_hoodStrings_02_l, loc=<3.2201 0 0>, rot=<0 0 0 1>}
 base_hoodStrings_01_r {parent=hoodStringsMaster, loc=<-0.5079 0.1312 -0.1352>, rot=<-0.5013 -0.563 0.531 -0.3871>}
 dyn_hoodStrings_01_r {parent=base_hoodStrings_01_r, loc=<0 0 0>, rot=<0 0 0 1>}
 dyn_hoodStrings_02_r {parent=dyn_hoodStrings_01_r, loc=<2.0513 0 0>, rot=<0 0 -0.0558 -0.9984>}
 dyn_hoodStrings_03_r {parent=dyn_hoodStrings_02_r, loc=<2.795 0 0>, rot=<0 0 0 1>}
 dyn_kennyUpperarm_l {parent=deform_spine1Baby, loc=<6.3464 -0.0973 -9.8274>, rot=<-0.56 -0.7342 -0.0517 -0.3804>}
 dyn_kennyLowerarm_l {parent=dyn_kennyUpperarm_l, loc=<4.8921 0 0>, rot=<0.0479 -0.0174 -0.3411 -0.9386>}
 dyn_kennyWrist_l {parent=dyn_kennyLowerarm_l, loc=<5.2762 0 0>, rot=<0 0 0 1>}
 dyn_kennyUpperarm_r {parent=deform_spine1Baby, loc=<6.3464 -0.0973 9.8274>, rot=<-0.3804 -0.0517 -0.7342 -0.56>}
 dyn_kennyLowerarm_r {parent=dyn_kennyUpperarm_r, loc=<4.8921 0 0>, rot=<-0.0479 -0.0174 0.3411 -0.9386>}
 dyn_kennyWrist_r {parent=dyn_kennyLowerarm_r, loc=<5.2762 0 0>, rot=<0 0 0 1>}
 attach_cape {parent=spine_05, loc=<9.017 -11.714 0>, rot=<0.0616 -0.7044 -0.0616 -0.7044>}
 attach_backpack {parent=spine_05, loc=<-0.123 -13.473 0>, rot=<-0.0093 -0.707 0.0093 -0.707>}
 handlerBar_Master {parent=spine_05, loc=<9.4085 23.5459 0.0133>, rot=<0.5196 -0.4795 0.5196 0.4795>}
 handlerBar_r {parent=handlerBar_Master, loc=<1.6895 0 0>, rot=<0 0 0 1>}
 base_tassle_01_r {parent=handlerBar_r, loc=<15.8704 -6.0088 3.7696>, rot=<0.0053 -0.0057 -0.3761 -0.9265>}
 dyn_tassle_01_r {parent=base_tassle_01_r, loc=<0 0 0>, rot=<0 0 0 1>}
 dyn_tassle_02_r {parent=dyn_tassle_01_r, loc=<2.9647 0.8742 0>, rot=<-0.0116 -0.0006 -0.3236 -0.9461>}
 dyn_tassle_03_r {parent=dyn_tassle_02_r, loc=<2.6093 0 0>, rot=<-0.0075 0.0141 0.0223 -0.9996>}
 handlerBar_l {parent=handlerBar_Master, loc=<-1.7162 0 0>, rot=<0 -1 0 0>}
 base_tassle_01_l {parent=handlerBar_l, loc=<15.8704 -6.0088 -3.7696>, rot=<0.9265 0.3761 0.0057 -0.0053>}
 dyn_tassle_01_l {parent=base_tassle_01_l, loc=<0 0 0>, rot=<0 0 0 1>}
 dyn_tassle_02_l {parent=dyn_tassle_01_l, loc=<2.9647 -0.8742 0>, rot=<0.0116 -0.0006 0.3236 -0.9461>}
 dyn_tassle_03_l {parent=dyn_tassle_02_l, loc=<2.6093 0 0>, rot=<0.0086 -0.0156 -0.0109 -0.9998>}
 dyn_arm_l {parent=spine_05, loc=<-1.075 -1.8984 -22.0546>, rot=<0.0568 -0.8743 -0.0198 -0.4816>}
 shoulderPad_Master_l {parent=dyn_arm_l, loc=<0 0 0>, rot=<0.47 -0.5388 0.541 0.4428>}
 shoulderPad_base_l {parent=shoulderPad_Master_l, loc=<4.4271 0 0>, rot=<0 0 0.0634 -0.998>}
 shoulderPad_01_l {parent=shoulderPad_base_l, loc=<13.8151 0 0>, rot=<0 0 0 1>}
 shoulderPad_02_l {parent=shoulderPad_01_l, loc=<0 0 0>, rot=<-0.003 -0.0003 0.029 -0.9996>}
 base_banana_01 {parent=shoulderPad_02_l, loc=<1.377 7.5939 -1.3915>, rot=<-0.0126 0.076 -0.1625 -0.9837>}
 dyn_banana_01 {parent=base_banana_01, loc=<0 0 0>, rot=<0 0 0 1>}
 dyn_banana_02 {parent=dyn_banana_01, loc=<5.4417 0 0>, rot=<0 0 0.0589 -0.9983>}
 base_bananaSide_01 {parent=shoulderPad_02_l, loc=<1.3819 -0.4795 -1.1454>, rot=<0.0225 0.0534 0.3879 -0.9199>}
 dyn_bananaSide_01 {parent=base_bananaSide_01, loc=<0 0 0>, rot=<0 0 0 1>}
 dyn_arm_aim_l {parent=dyn_arm_l, loc=<0 0 0>, rot=<0 0 0 1>}
 dyn_arm_r {parent=spine_05, loc=<-1.075 -1.8984 22.0546>, rot=<0.8763 0.055 -0.4781 0.0205>}
 shoulderPad_Master_r {parent=dyn_arm_r, loc=<0 0 0>, rot=<0.5568 -0.4243 -0.4614 -0.5451>}
 shoulderPad_base_r {parent=shoulderPad_Master_r, loc=<4.4271 0 0>, rot=<0 0 -0.0634 -0.998>}
 shoulderPad_01_r {parent=shoulderPad_base_r, loc=<10.3463 0.6047 0.0006>, rot=<0 0 0 1>}
 shoulderPad_02_r {parent=shoulderPad_01_r, loc=<0 0 0>, rot=<0.003 -0.0002 0.0215 -0.9998>}
 dyn_arm_aim_r {parent=dyn_arm_r, loc=<0 0 0>, rot=<0 0 0 1>}
 thigh_l {parent=pelvis, loc=<-1.8716 -0.0109 -11.9936>, rot=<0.0012 0.0572 0.0155 -0.9982>}
 calf_l {parent=thigh_l, loc=<-41.9178 -0.2246 0.0926>, rot=<0.0006 -0.0161 -0.0391 -0.9991>}
 foot_l {parent=calf_l, loc=<-38.7057 1.2438 0.2452>, rot=<0.0975 -0.0474 0.0317 -0.9936>}
 ball_l {parent=foot_l, loc=<-6.1614 13.1469 0.6807>, rot=<0.0001 -0.0001 -0.7063 -0.7079>}
 calf_twist_02_l {parent=calf_l, loc=<-15.9268 -0.0967 0.0397>, rot=<0.0237 0.0095 -0.0035 -0.9997>}
 calf_twist_01_l {parent=calf_l, loc=<-32.9335 -0.0967 0.0397>, rot=<0.0237 0.0095 -0.0035 -0.9997>}
 deform_knee_in_base_l {parent=calf_l, loc=<1.298 -8.0356 -0.0496>, rot=<0 0 0 -1>}
 deform_knee_in_l {parent=deform_knee_in_base_l, loc=<0 0 0>, rot=<0 0 0 1>}
 deform_knee_out_base_l {parent=calf_l, loc=<0 6.8739 0>, rot=<0 0 0 -1>}
 deform_knee_out_l {parent=deform_knee_out_base_l, loc=<0 0 0>, rot=<0 0 0 1>}
 thigh_twist_01_l {parent=thigh_l, loc=<-21.6303 0 0>, rot=<0.0013 0 0 -1>}
 thighArmorMover_l {parent=thigh_l, loc=<-18.5973 0 0>, rot=<0.0013 0 0 -1>}
 thigh_r {parent=pelvis, loc=<-1.8716 -0.0109 11.9936>, rot=<0.0572 -0.0012 0.9982 0.0155>}
 calf_r {parent=thigh_r, loc=<41.9178 0.2246 -0.0926>, rot=<0.0006 -0.0161 -0.0391 -0.9991>}
 foot_r {parent=calf_r, loc=<38.7057 -1.2438 -0.2452>, rot=<0.0975 -0.0474 0.0317 -0.9936>}
 ball_r {parent=foot_r, loc=<6.1614 -13.1469 -0.6807>, rot=<0.0001 -0.0001 -0.7063 -0.7079>}
 calf_twist_02_r {parent=calf_r, loc=<15.9268 0.0967 -0.0397>, rot=<0.0237 0.0095 -0.0035 -0.9997>}
 calf_twist_01_r {parent=calf_r, loc=<32.9335 0.0967 -0.0397>, rot=<0.0237 0.0095 -0.0035 -0.9997>}
 deform_knee_in_base_r {parent=calf_r, loc=<-1.2981 8.0356 0.0496>, rot=<0 0 0 -1>}
 deform_knee_in_r {parent=deform_knee_in_base_r, loc=<0 0 0>, rot=<0 0 0 1>}
 deform_knee_out_base_r {parent=calf_r, loc=<0 -6.8739 0>, rot=<0 0 0 -1>}
 deform_knee_out_r {parent=deform_knee_out_base_r, loc=<0 0 0>, rot=<0 0 0 1>}
 base_bagFlap_01_l {parent=calf_r, loc=<5.1824 1.4082 -8.2779>, rot=<0.2868 -0.6496 -0.6358 -0.3023>}
 dyn_bagFlap_01_l {parent=base_bagFlap_01_l, loc=<0 0 0>, rot=<0 0 0 1>}
 dyn_bagFlap_02_l {parent=dyn_bagFlap_01_l, loc=<3.9595 0 0>, rot=<0 0 0.2016 -0.9795>}
 base_bagFlap_01_r {parent=calf_r, loc=<6.813 3.0297 9.5854>, rot=<0.327 0.6329 0.6601 -0.2383>}
 dyn_bagFlap_01_r {parent=base_bagFlap_01_r, loc=<0 0 0>, rot=<0 0 0 1>}
 dyn_bagFlap_02_r {parent=dyn_bagFlap_01_r, loc=<4.1074 0 0>, rot=<0 0 -0.0885 -0.9961>}
 thigh_twist_01_r {parent=thigh_r, loc=<21.6303 0 0>, rot=<0.0013 0 0 -1>}
 thighArmorMover_r {parent=thigh_r, loc=<18.597 0 0>, rot=<0.0013 0 0 -1>}
 attach {parent=root, loc=<0 0 0>, rot=<0 0 0 1>}
 ik_hand_root {parent=root, loc=<0 0 0>, rot=<0 0 0 1>}
 ik_hand_gun {parent=ik_hand_root, loc=<-64.462 4.8638 101.1961>, rot=<0.6084 0.1793 0.4613 -0.6205>}
 ik_hand_r {parent=ik_hand_gun, loc=<0 0 0>, rot=<0 0 0 1>}
 ik_hand_l {parent=ik_hand_gun, loc=<65.7708 101.9258 43.6659>, rot=<-0.5102 -0.7906 -0.3387 0>}
 ik_foot_root {parent=root, loc=<0 0 0>, rot=<0 0 0 1>}
 ik_foot_l {parent=ik_foot_root, loc=<19.6189 -3.2051 10.3339>, rot=<0.0753 0.7018 0.0677 -0.7051>}
 ik_foot_r {parent=ik_foot_root, loc=<-19.6189 -3.2051 10.3339>, rot=<0.7051 0.0677 -0.7018 0.0753>}
 root_fp {parent=root, loc=<0 0 0>, rot=<0 0 0 1>}
 camera_root {parent=root_fp, loc=<0 0 0>, rot=<0 0 0 1>}
 camera {parent=camera_root, loc=<0 0 0>, rot=<0 0 0 1>}
 attach_fp {parent=root_fp, loc=<0 0 0>, rot=<0 0 0 1>}
 attach_emote_01 {parent=root, loc=<0 0 0>, rot=<0 0 0 1>}
 attach_emote_02 {parent=root, loc=<0 0 0>, rot=<0 0 0 1>}
 attach_emote_03 {parent=root, loc=<0 0 0>, rot=<0 0 0 1>}
 warp_point_01 {parent=root, loc=<0 0 0>, rot=<0 0 0 1>}
 warp_point_02 {parent=root, loc=<0 0 0>, rot=<0 0 0 1>}
 warp_point_03 {parent=root, loc=<0 0 0>, rot=<0 0 0 1>}
]
variables [
 Char Backwards {type=double, default=0}
 Char Forward {type=double, default=0}
 Char Left {type=double, default=0}
 Char Right {type=double, default=0}
 Head Down {type=double, default=0}
 Head Left {type=double, default=0}
 Head Right {type=double, default=0}
]
