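# Physics asset for the face accessory part of the large male PacificSweater player cosmetic, covering mainly the hood strings.
# Capsule bodies cover head, neck, wrist and handlerBar bones, plus a base and two dyn segments per string.
# Each string chains from its base through tight-twist, limited-swing constraints.
# Neck-to-head joints have free rotation, while wrist, handlerBar and string base bodies have no constraints to the rest.
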PhysicsAsset /BRCosmetics/Characters/Player/Male/Large/Bodies/M_LRG_PacificSweater/Meshes/Parts/M_LRG_PacificSweater_FaceAcc_Physics
bodies {
 base_hoodStrings_01_l [{shapeType=Capsule, length=0.505, radius=0.505, shapeName=base_hoodStrings_01_l_capsule, rotation=<rot 0 -90 0>}]
 base_hoodStrings_01_r [{shapeType=Capsule, length=0.505, radius=0.505, shapeName=base_hoodStrings_01_r_capsule, rotation=<rot 0 -90 0>}]
 dyn_hoodStrings_01_l [{
   shapeType=Capsule
   length=1.4454
   radius=0.8257
   shapeName=dyn_hoodStrings_01_r_capsule
   center=<1.3038 -0.083 0>
   rotation=<rot 90 2.8786 0>
  }]
 dyn_hoodStrings_01_r [{shapeType=Capsule, length=1.528, radius=0.743, shapeName=dyn_hoodStrings_01_r_capsule, center=<1.0795 -0.0942 0>, rotation=<rot 90 2.8786 0>}]
 dyn_hoodStrings_02_l [{shapeType=Capsule, length=1.8341, radius=0.8745, shapeName=dyn_hoodStrings_02_r_capsule, center=<1.7327 0.2334 0>, rotation=<rot 90 0 0>}]
 dyn_hoodStrings_02_r [{shapeType=Capsule, length=1.8341, radius=0.8745, shapeName=dyn_hoodStrings_02_r_capsule, center=<1.3239 -0.2439 0>, rotation=<rot 90 0 0>}]
 dyn_kennyWrist_l [{
   shapeType=Capsule
   length=3.355
   radius=3.648
   shapeName=dyn_kennyWrist_l_capsule
   center=<1.8141 0.1511 -0.1674>
   rotation=<rot 43.131 -88.0648 -88.677>
  }]
 dyn_kennyWrist_r [{
   shapeType=Capsule
   length=3.355
   radius=3.648
   shapeName=dyn_kennyWrist_l_capsule
   center=<1.8141 0.1511 -0.1674>
   rotation=<rot 43.131 -88.0648 -88.677>
  }]
 handlerBar_l [{
   shapeType=Capsule
   length=7.6832
   radius=1.3724
   shapeName=handlerBar_l_capsule
   center=<4.1525 -3.4476 -1.9515>
   rotation=<rot 61.1918 -90 -15.3705>
  }]
 handlerBar_r [{
   shapeType=Capsule
   length=7.6832
   radius=1.3724
   shapeName=handlerBar_l_capsule
   center=<4.1525 -3.4476 1.9515>
   rotation=<rot -61.1918 -90 15.3705>
  }]
 head [{shapeType=Capsule, length=5, radius=5, shapeName=head_capsule, center=<9.4835 10.0607 -0.0002>, rotation=<rot 0 -9.5698 -89.9983>}]
 neck_01 [{
   shapeType=Capsule
   length=20.6357
   radius=8.3241
   shapeName=neck_01_capsule
   center=<3.061 4.8904 -0.0008>
   rotation=<rot -0.0144 -90 0.0082>
  }]
 neck_02 [{shapeType=Capsule, length=0.505, radius=0.505, shapeName=neck_02_capsule, rotation=<rot 0 -90 0>}]
}
constraints [
 {
  bone1Name=neck_02
  bone2Name=neck_01
  swing1LimitDegrees=45
  swing1Motion=Free
  swing2LimitDegrees=45
  swing2Motion=Free
  twistLimitDegrees=45
  twistMotion=Free
 }
 {
  bone1Name=head
  bone2Name=neck_02
  swing1LimitDegrees=45
  swing1Motion=Free
  swing2LimitDegrees=45
  swing2Motion=Free
  twistLimitDegrees=45
  twistMotion=Free
 }
 {bone1Name=dyn_hoodStrings_01_l, bone2Name=base_hoodStrings_01_l, swing1LimitDegrees=25, swing2LimitDegrees=25, twistLimitDegrees=1}
 {bone1Name=dyn_hoodStrings_02_l, bone2Name=dyn_hoodStrings_01_l, swing1LimitDegrees=25, swing2LimitDegrees=25, twistLimitDegrees=1}
 {bone1Name=dyn_hoodStrings_01_r, bone2Name=base_hoodStrings_01_r, swing1LimitDegrees=25, swing2LimitDegrees=25, twistLimitDegrees=1}
 {bone1Name=dyn_hoodStrings_02_r, bone2Name=dyn_hoodStrings_01_r, swing1LimitDegrees=25, swing2LimitDegrees=25, twistLimitDegrees=1}
]
